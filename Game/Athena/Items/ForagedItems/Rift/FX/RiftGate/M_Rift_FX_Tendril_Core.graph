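# Unlit translucent Niagara ribbon material for the core of a Rift gate tendril effect.
# Panning noise blends two colors per stage, with an Overload dynamic parameter shifting stages and an optional rainbow texture override.
# A gradient positioned by Pulse_Offset adds a bright pulse to the emissive.
# Opacity layers noise-distorted vertical gradients, clipped along U by a Mask parameter and depth-faded except on mobile.
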Material /Game/Athena/Items/ForagedItems/Rift/FX/RiftGate/M_Rift_FX_Tendril_Core
graph {
 properties {bUsedWithNiagaraRibbons=true, blendMode=BLEND_Translucent, pixelDepthOffsetMode=PDOM_Legacy, shadingModel=MSM_Unlit, usedShadingModels=Unlit}
 nodes [
  n1 = DynamicParameter() {defaultValue=<rgba 1 1 0 1>, paramNames [Mask, Thread_Tiling_X, Noise_Tiling_X, Param4]}
  n2 = Constant() {r=-0.5}
  n3 = Subtract(A=TextureCoordinate(), B=AppendVector(A=Add(A=@n1.Mask, B=@n2), B=Constant()))
  n4 = Clamp(None=Multiply(A=ComponentMask(None=@n3) {r=true}, B=Constant() {r=2}))
  Floor(None=@n4)
  n5 = DynamicParameter() {parameterIndex=1, defaultValue=<rgba 0 1 0 0.3>, paramNames [Overload, Thread_Tiling_X, Noise_Tiling_X, Pulse_Offset]}
  n6 = TextureCoordinate()
  n7 = AppendVector(A=Constant() {r=1}, B=Constant() {r=1})
  n8 = DynamicParameter() {parameterIndex=1, defaultValue=<rgba 1 1 0.5 0>, paramNames [Mask, Thread_Tiling_X, Pulse_Offset, ColorNoiseOffset]}
  n9 = LinearInterpolate(Alpha=@n8.Pulse_Offset) {constA=-0.6, constB=0.6}
  n10 = Constant()
  n11 = TextureSample(UVs=Add(A=Multiply(A=@n6, B=@n7), B=AppendVector(A=@n9, B=@n10))) {samplerSource=SSM_Clamp_WorldGroupSettings, texture=/Game/Athena/Items/ForagedItems/Rift/FX/RiftGate/Textures/T_LinearGrad_Horizontal_Rift}
  n12 = Multiply(
   A=LinearInterpolate(
    A=VectorParameter() {parameterName=Pulse_Color_Stage1, defaultValue=<rgba 0 0.3261 1 1>}.RGB
    B=VectorParameter() {parameterName=Pulse_Color_Stage2, defaultValue=<rgba 5 3.3333 2.5 1>}.RGB
    Alpha=@n5.Overload
   )
   B=@n11.R
  )
  n13 = ScalarParameter() {defaultValue=1, parameterName=Pulse_Power}
  n14 = VectorParameter() {parameterName=Color_A_Stage1, defaultValue=<rgba 0 0.3261 1 1>}
  n15 = VectorParameter() {parameterName=Color_B_Stage1, defaultValue=<rgba 0.781 0 1 1>}
  n16 = DynamicParameter() {parameterIndex=1, defaultValue=<rgba 1 1 0.5 0>, paramNames [Mask, Thread_Tiling_X, Pulse_Offset, ColorNoiseOffset]}
  n17 = TextureCoordinate()
  n18 = AppendVector(A=ScalarParameter() {defaultValue=0.5, parameterName=ColorGrad_Tile_X}, B=Constant())
  n19 = ScalarParameter() {defaultValue=0.1, parameterName=ColorGrad_Pan_X}
  n20 = ScalarParameter() {defaultValue=0.01, parameterName=ColorGrad_Pan_Y}
  n21 = Add(A=AppendVector(A=@n16.ColorNoiseOffset, B=Constant()), B=Panner(Coordinate=Multiply(A=@n17, B=@n18), Speed=AppendVector(A=@n19, B=@n20)))
  n22 = TextureSample(UVs=@n21) {samplerType=SAMPLERTYPE_LinearColor, texture=&HighContrastNoise}
  n23 = ScalarParameter() {defaultValue=4, parameterName=ColorGrad_Power}
  n24 = Clamp(None=Multiply(A=Power(Base=@n22.R, Exp=@n23), B=ScalarParameter() {defaultValue=2, parameterName=ColorGrad_Mult}))
  n25 = VectorParameter() {parameterName=Color_A_Stage2, defaultValue=<rgba 1 0.3333 0.2 1>}
  n26 = VectorParameter() {parameterName=Color_B_Stage2, defaultValue=<rgba 1 0.4667 0.2 1>}
  n27 = DynamicParameter() {parameterIndex=1, defaultValue=<rgba 0 1 0 0.3>, paramNames [Overload, Thread_Tiling_X, Noise_Tiling_X, Pulse_Offset]}
  n28 = LinearInterpolate(
   A=LinearInterpolate(A=LinearInterpolate(A=@n14.RGB, B=@n15.RGB, Alpha=@n24), B=LinearInterpolate(A=@n25.RGB, B=@n26.RGB, Alpha=@n24), Alpha=@n27.Overload)
   B=TextureSample(UVs=@n21) {texture=/Game/Athena/Items/ForagedItems/Rift/FX/RiftGate/Textures/T_Tendril_Rainbow}.RGB
   Alpha=ScalarParameter() {parameterName=Use_Rainbow}
  )
  n29 = Multiply(A=Multiply(A=Power(Base=@n12, Exp=@n13), B=ScalarParameter() {defaultValue=30, parameterName=Pulse_Mult}), B=@n28)
  n30 = ParticleColor()
  MaterialFunctionCall(Emissive=Multiply(A=Add(A=@n29, B=@n28), B=@n30.RGB)) {materialFunction=&MF_FX_EmissiveAdjustments}
  n31 = TextureCoordinate() {coordinateIndex=1}
  n32 = Multiply(A=TextureCoordinate() {coordinateIndex=1}, B=AppendVector(A=Constant() {r=1}, B=Constant() {r=0.5}))
  n33 = AppendVector(A=Constant() {r=0.2}, B=Constant() {r=0.0125})
  n34 = Multiply(
   A=TextureSample(UVs=Panner(Coordinate=@n32, Speed=@n33)) {texture=&AtmosphericCloudNoise01}.R
   B=Constant() {r=0.25}
  ) {constB=0.5}
  n35 = Constant()
  n36 = Constant()
  n37 = TextureSample(UVs=Add(A=Add(A=@n31, B=@n34), B=AppendVector(A=@n35, B=@n36)) {constB=-0.5}) {
   samplerSource=SSM_Clamp_WorldGroupSettings
   samplerType=SAMPLERTYPE_LinearColor
   texture=/Game/Accessories/FORT_Backpacks/Backpack_F_MED_Palespooky_Outlander/Skins/Pancake/Textures/T_Palespooky_Pancake_VertGradient
  }
  n38 = Clamp(None=Multiply(A=Power(Base=@n37.RGB) {constExponent=4}))
  n39 = Multiply(A=TextureCoordinate() {coordinateIndex=1}, B=AppendVector(A=Constant() {r=0.1}, B=Constant() {r=0.5}))
  n40 = AppendVector(A=Constant() {r=0.05}, B=Constant() {r=0.0125})
  n41 = Power(Base=TextureSample(UVs=Panner(Coordinate=@n39, Speed=@n40)) {texture=&AtmosphericCloudNoise01}.RGB) {constExponent=4}
  n42 = Clamp(None=LinearInterpolate(A=Constant() {r=0.1}, B=Constant() {r=10}, Alpha=Multiply(A=@n41) {constB=2}))
  n43 = TextureCoordinate() {coordinateIndex=1}
  n44 = Multiply(A=TextureCoordinate() {coordinateIndex=1}, B=AppendVector(A=Constant() {r=1}, B=Constant() {r=0.5}))
  n45 = AppendVector(A=Constant() {r=0.2}, B=Constant() {r=0.0125})
  n46 = Multiply(
   A=TextureSample(UVs=Panner(Coordinate=@n44, Speed=@n45)) {texture=&AtmosphericCloudNoise01}.R
   B=Constant() {r=0.25}
  ) {constB=0.5}
  n47 = Constant()
  n48 = Constant()
  n49 = TextureSample(UVs=Add(A=Add(A=@n43, B=@n46), B=AppendVector(A=@n47, B=@n48)) {constB=-0.5}) {
   samplerSource=SSM_Clamp_WorldGroupSettings
   samplerType=SAMPLERTYPE_LinearColor
   texture=&T_Palespooky_Pancake_VertGradient
  }
  n50 = Multiply(A=Power(Base=@n49.RGB) {constExponent=10})
  n51 = Multiply(A=Add(A=Multiply(A=@n42, B=@n38), B=Clamp(None=@n50)), B=OneMinus(None=@n4))
  n52 = DepthFade(Opacity=@n51, FadeDistance=ScalarParameter() {defaultValue=1000, parameterName=DepthFade Distance})
  Multiply(A=@n30.A, B=ShadingPathSwitch(Default=@n52, Deferred=@n52, Mobile=@n51) {desc=Migrated from FeatureLevelSwitch})
 ]
}
