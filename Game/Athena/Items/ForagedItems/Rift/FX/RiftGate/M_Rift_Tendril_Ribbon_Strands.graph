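# Unlit, two-sided translucent material for Rift Gate tendril strands on Niagara ribbons and spline meshes.
# Up to three panning strand layers, plus optional caustic noise, are tinted by noise-driven stage colors or a rainbow texture.
# Dynamic parameters drive reveal, pan speed and Overload color/texture blends; the channel named Noise_Tiling_X slides the pulse gradient.
# Opacity combines the reveal with optional sine soft mask, Fresnel and depth fade, which Mobile skips.
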Material /Game/Athena/Items/ForagedItems/Rift/FX/RiftGate/M_Rift_Tendril_Ribbon_Strands
graph {
 properties {
  bUsedWithNiagaraRibbons=true
  bUsedWithSplineMeshes=true
  blendMode=BLEND_Translucent
  pixelDepthOffsetMode=PDOM_Legacy
  shadingModel=MSM_Unlit
  twoSided=true
  usedShadingModels=Unlit
 }
 nodes [
  DynamicParameter() {parameterIndex=1, defaultValue=<rgba 0 1 0 0.3>, paramNames [Overload, Thread_Tiling_X, Noise_Tiling_X, Pulse_Offset]}
  DynamicParameter() {defaultValue=<rgba 1 1 0 0.3>, paramNames [Mask, Thread_Tiling_X, Noise_Tiling_X, PanSpeed]}
  n1 = Add(
   A=ScalarParameter() {parameterName=TestingMaskReveal}
   B=DynamicParameter() {defaultValue=<rgba 1 1 0 0.3>, paramNames [Mask, Thread_Tiling_X, Noise_Tiling_X, PanSpeed]}.Mask
  )
  n2 = Constant() {r=-0.5}
  n3 = Subtract(A=TextureCoordinate(), B=AppendVector(A=Add(A=@n1, B=@n2), B=Constant()))
  n4 = Clamp(None=Multiply(A=ComponentMask(None=@n3) {r=true}, B=Constant() {r=2}))
  Floor(None=@n4)
  n5 = Multiply(
   A=LinearInterpolate(A=TextureCoordinate(), B=TextureCoordinate() {coordinateIndex=1}, Alpha=ScalarParameter() {parameterName=Use_UV1})
   B=AppendVector(A=ScalarParameter() {defaultValue=6, parameterName=HeroLine_A_Tile_X}, B=Constant() {r=1})
  )
  n6 = AppendVector(A=Constant() {r=0.25}, B=Constant())
  n7 = Add(
   A=ScalarParameter() {defaultValue=0.1, parameterName=HeroLine_A_Pan_X}
   B=DynamicParameter() {defaultValue=<rgba 1 1 0 0.3>, paramNames [Mask, Thread_Tiling_X, Noise_Tiling_X, PanSpeed]}.PanSpeed
  )
  n8 = Constant()
  n9 = MaterialFunctionCall(XY=Add(A=Add(A=@n5, B=@n6), B=AppendVector(A=@n7, B=@n8))) {materialFunction=&Swizzle}
  n10 = DynamicParameter() {parameterIndex=1, defaultValue=<rgba 0 1 0 0.3>, paramNames [Overload, Thread_Tiling_X, Noise_Tiling_X, Pulse_Offset]}
  n11 = Constant()
  n12 = Constant() {r=1}
  n13 = Constant()
  n14 = Constant() {r=0.5}
  n15 = Multiply(
   A=LinearInterpolate(
    A=TextureSample(UVs=@n9.YX) {texture=/Game/Athena/Items/ForagedItems/Rift/FX/RiftGate/Textures/Rift_Tendril_Strands}.R
    B=TextureSample(UVs=@n9.YX) {texture=/Game/Athena/Items/ForagedItems/Rift/FX/RiftGate/Textures/T_RiftGate_Portal_Overload_E}.R
    Alpha=MaterialFunctionCall(Input=@n10.Overload, Input Low=@n11, Input High=@n12, Target Low=@n13, Target High=@n14) {materialFunction=&RemapValueRange}.Result
   )
   B=Constant() {r=1}
  )
  n16 = Multiply(
   A=LinearInterpolate(A=TextureCoordinate(), B=TextureCoordinate() {coordinateIndex=1}, Alpha=ScalarParameter() {parameterName=Use_UV1})
   B=AppendVector(A=ScalarParameter() {defaultValue=10, parameterName=HeroLine_B_Tile_X}, B=Constant() {r=-1})
  )
  n17 = AppendVector(A=Constant(), B=Constant() {r=1})
  n18 = Add(
   A=ScalarParameter() {defaultValue=0.2, parameterName=HeroLine_B_Pan_X}
   B=DynamicParameter() {defaultValue=<rgba 1 1 0 0.3>, paramNames [Mask, Thread_Tiling_X, Noise_Tiling_X, PanSpeed]}.PanSpeed
  )
  n19 = Constant()
  n20 = MaterialFunctionCall(XY=Add(A=Add(A=@n16, B=@n17), B=AppendVector(A=@n18, B=@n19))) {materialFunction=&Swizzle}
  n21 = TextureSample(UVs=@n20.YX) {texture=&Rift_Tendril_Strands}
  n22 = TextureSample(UVs=@n20.YX) {texture=&T_RiftGate_Portal_Overload_E}
  n23 = DynamicParameter() {parameterIndex=1, defaultValue=<rgba 0 1 0 0.3>, paramNames [Overload, Thread_Tiling_X, Noise_Tiling_X, Pulse_Offset]}
  n24 = MaterialFunctionCall(Input=@n23.Overload, Input Low=Constant(), Input High=Constant() {r=1}, Target Low=Constant(), Target High=Constant() {r=0.5}) {materialFunction=&RemapValueRange}
  n25 = Add(A=@n15, B=Multiply(A=LinearInterpolate(A=@n21.G, B=@n22.G, Alpha=@n24.Result), B=Constant() {r=1}))
  n26 = Multiply(
   A=LinearInterpolate(A=TextureCoordinate(), B=TextureCoordinate() {coordinateIndex=1}, Alpha=ScalarParameter() {parameterName=Use_UV1})
   B=AppendVector(A=ScalarParameter() {defaultValue=5, parameterName=HeroLine_C_Tile_X}, B=Constant() {r=1})
  )
  n27 = AppendVector(A=Constant() {r=0.8}, B=Constant())
  n28 = DynamicParameter() {defaultValue=<rgba 1 1 0 0.3>, paramNames [Mask, Thread_Tiling_X, Noise_Tiling_X, PanSpeed]}
  n29 = Constant()
  n30 = MaterialFunctionCall(XY=Add(A=Add(A=@n26, B=@n27), B=AppendVector(A=@n28.PanSpeed, B=@n29))) {materialFunction=&Swizzle}
  n31 = Multiply(
   A=LinearInterpolate(
    A=TextureSample(UVs=@n30.YX) {texture=&Rift_Tendril_Strands}.B
    B=TextureSample(UVs=@n30.YX) {texture=&T_RiftGate_Portal_Overload_E}.B
    Alpha=ScalarParameter() {defaultValue=0.5, parameterName=softentexture}
   )
   B=Constant() {r=1}
  )
  n32 = LinearInterpolate(
   A=LinearInterpolate(A=@n15, B=Add(A=@n25, B=@n31), Alpha=ScalarParameter() {defaultValue=1, parameterName=Use_MultiStrands})
   B=@n25
   Alpha=ScalarParameter() {parameterName=Use_2Strands}
  )
  n33 = Add(A=@n32, B=Multiply(A=@n32))
  n34 = Constant() {r=0.25}
  n35 = LinearInterpolate(A=TextureCoordinate(), B=TextureCoordinate() {coordinateIndex=1}, Alpha=ScalarParameter() {parameterName=Use_UV1})
  n36 = AppendVector(
   A=ScalarParameter() {defaultValue=0.1, parameterName=SoftNoise_Tile_X}
   B=ScalarParameter() {defaultValue=0.1, parameterName=SoftNoise_Tile_Y}
  )
  n37 = ScalarParameter() {parameterName=SoftNoise_Pan_X}
  n38 = Constant()
  n39 = TextureSample(UVs=Panner(Coordinate=Multiply(A=@n35, B=@n36), Speed=AppendVector(A=@n37, B=@n38))) {texture=/Game/Global/Textures/Default/Caustics/T_Caustics_Softer}
  n40 = LinearInterpolate(A=@n33, B=Add(A=@n33, B=Multiply(A=@n39.R, B=@n34)), Alpha=ScalarParameter() {parameterName=Use_Noise})
  n41 = ParticleColor()
  n42 = LinearInterpolate(A=Constant3Vector() {constant=<rgba 0 1.9563 6 1>}, B=@n41.RGB, Alpha=ScalarParameter() {defaultValue=1, parameterName=Use_ParticleColor})
  n43 = DynamicParameter() {parameterIndex=1, defaultValue=<rgba 0 1 0 0.3>, paramNames [Overload, Thread_Tiling_X, Noise_Tiling_X, Pulse_Offset]}
  n44 = TextureCoordinate()
  n45 = AppendVector(A=Constant() {r=1}, B=Constant() {r=1})
  n46 = DynamicParameter() {parameterIndex=1, defaultValue=<rgba 0 1 0 0.3>, paramNames [Overload, Thread_Tiling_X, Noise_Tiling_X, Pulse_Offset]}
  n47 = LinearInterpolate(Alpha=@n46.Noise_Tiling_X) {constA=-0.6, constB=0.6}
  n48 = Constant()
  n49 = TextureSample(UVs=Add(A=Multiply(A=@n44, B=@n45), B=AppendVector(A=@n47, B=@n48))) {samplerSource=SSM_Clamp_WorldGroupSettings, texture=/Game/Athena/Items/ForagedItems/Rift/FX/RiftGate/Textures/T_LinearGrad_Horizontal_Rift}
  n50 = Multiply(
   A=LinearInterpolate(
    A=VectorParameter() {parameterName=Pulse_Color_Stage1, defaultValue=<rgba 0 0.3261 1 1>}.RGB
    B=VectorParameter() {parameterName=Pulse_Color_Stage2, defaultValue=<rgba 5 3.3333 2.5 1>}.RGB
    Alpha=@n43.Overload
   )
   B=@n49.R
  )
  n51 = ScalarParameter() {defaultValue=1, parameterName=Pulse_Power}
  n52 = VectorParameter() {parameterName=Color_A_Stage1, defaultValue=<rgba 0 0.3261 1 1>}
  n53 = VectorParameter() {parameterName=Color_B_Stage1, defaultValue=<rgba 0.781 0 1 1>}
  n54 = DynamicParameter() {parameterIndex=1, defaultValue=<rgba 0 1 0 0.3>, paramNames [Overload, Thread_Tiling_X, Noise_Tiling_X, Pulse_Offset]}
  n55 = TextureCoordinate()
  n56 = AppendVector(A=ScalarParameter() {defaultValue=0.5, parameterName=ColorGrad_Tile_X}, B=Constant())
  n57 = ScalarParameter() {defaultValue=0.1, parameterName=ColorGrad_Pan_X}
  n58 = ScalarParameter() {defaultValue=0.01, parameterName=ColorGrad_Pan_Y}
  n59 = Add(A=AppendVector(A=@n54.Pulse_Offset, B=Constant()), B=Panner(Coordinate=Multiply(A=@n55, B=@n56), Speed=AppendVector(A=@n57, B=@n58)))
  n60 = TextureSample(UVs=@n59) {samplerType=SAMPLERTYPE_LinearColor, texture=&HighContrastNoise}
  n61 = ScalarParameter() {defaultValue=4, parameterName=ColorGrad_Power}
  n62 = Clamp(None=Multiply(A=Power(Base=@n60.R, Exp=@n61), B=ScalarParameter() {defaultValue=2, parameterName=ColorGrad_Mult}))
  n63 = VectorParameter() {parameterName=Color_A_Stage2, defaultValue=<rgba 1 0.3333 0.2 1>}
  n64 = VectorParameter() {parameterName=Color_B_Stage2, defaultValue=<rgba 1 0.4667 0.2 1>}
  n65 = DynamicParameter() {parameterIndex=1, defaultValue=<rgba 0 1 0 0.3>, paramNames [Overload, Thread_Tiling_X, Noise_Tiling_X, Pulse_Offset]}
  n66 = LinearInterpolate(
   A=LinearInterpolate(A=LinearInterpolate(A=@n52.RGB, B=@n53.RGB, Alpha=@n62), B=LinearInterpolate(A=@n63.RGB, B=@n64.RGB, Alpha=@n62), Alpha=@n65.Overload)
   B=TextureSample(UVs=@n59) {texture=/Game/Athena/Items/ForagedItems/Rift/FX/RiftGate/Textures/T_Tendril_Rainbow}.RGB
   Alpha=ScalarParameter() {parameterName=Use_Rainbow}
  )
  n67 = Multiply(A=Multiply(A=Power(Base=@n50, Exp=@n51), B=ScalarParameter() {defaultValue=10, parameterName=Pulse_Mult}), B=@n66)
  MaterialFunctionCall(Emissive=Multiply(A=Add(A=@n67, B=@n66), B=Multiply(A=@n40, B=@n42))) {materialFunction=&MF_FX_EmissiveAdjustments}
  Multiply(A=@n39.R, B=Constant() {r=2})
  n68 = OneMinus(None=@n4)
  n69 = TextureCoordinate() {vTiling=0.5}
  n70 = TextureCoordinate() {coordinateIndex=1, uTiling=0.5}
  n71 = ScalarParameter() {parameterName=Use_UV1}
  n72 = Sine(None=ComponentMask(None=LinearInterpolate(A=@n69, B=@n70, Alpha=@n71)) {g=true})
  n73 = Multiply(
   A=Power(Base=Multiply(A=@n72), Exp=ScalarParameter() {defaultValue=3, parameterName=SoftMask_Power})
   B=ScalarParameter() {defaultValue=1, parameterName=SoftMask_Mult}
  )
  n74 = Multiply(
   A=LinearInterpolate(A=Constant() {r=1}, B=@n41.A, Alpha=ScalarParameter() {defaultValue=1, parameterName=Use_ParticleColor})
   B=LinearInterpolate(A=Multiply(A=@n40, B=@n68), B=Multiply(A=@n68, B=@n73), Alpha=ScalarParameter() {parameterName=use_SoftMask})
  )
  n75 = Constant() {r=1}
  n76 = Constant() {r=1}
  n77 = LinearInterpolate(
   A=@n74
   B=Multiply(
    A=MaterialFunctionCall(Brightness=@n75, Exponent=@n76) {materialFunction=&MF_Fresnel_FX}.Fresnel
    B=@n74
   )
   Alpha=ScalarParameter() {defaultValue=1, parameterName=Use_Fresnel}
  )
  n78 = DepthFade(Opacity=@n77, FadeDistance=ScalarParameter() {defaultValue=1000, parameterName=DepthFade Distance})
  Reroute(None=ShadingPathSwitch(Default=@n78, Deferred=@n78, Mobile=@n77) {desc=Migrated from FeatureLevelSwitch})
  ScalarParameter() {defaultValue=0.2, parameterName=HeroLine_C_Pan_X}
 ]
}
