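# Unlit, two-sided translucent material for a light cone effect on the Rift Gate, used on Niagara mesh particles.
# World position offset pushes vertices along their normals, blending bottom and top radius parameters by vertex color red to shape the cone.
# Opacity multiplies UV-based end fades, particle alpha, panning noise and double-sided fresnel, with depth fade skipped on mobile.
# Emissive color comes straight from particle color.
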Material /Game/Athena/Items/ForagedItems/Rift/FX/RiftGate/M_Riftgate_Lightcone
graph {
 properties {
  bUsedWithNiagaraMeshParticles=true
  blendMode=BLEND_Translucent
  pixelDepthOffsetMode=PDOM_Legacy
  shadingModel=MSM_Unlit
  twoSided=true
  usedShadingModels=Unlit
 }
 nodes [
  n1 = ScalarParameter() {defaultValue=100, parameterName=WPO_Radius_Bottom}
  n2 = ScalarParameter() {defaultValue=1000, parameterName=WPO_Radius_Top}
  n3 = VertexColor()
  NamedRerouteDeclaration(None=Multiply(A=VertexNormalWS(), B=LinearInterpolate(A=@n1, B=@n2, Alpha=@n3.R))) {name=WPO, nodeColor=<rgba 1 0 0.6328 1>}
  n4 = Power(Base=ComponentMask(None=TextureCoordinate()) {g=true}, Exp=ScalarParameter() {defaultValue=2.2, parameterName=BottomEdgeFade_Power})
  n5 = ScalarParameter() {defaultValue=1, parameterName=BottomEdgeFade_Mult}
  n6 = ComponentMask(None=TextureCoordinate()) {g=true}
  n7 = Multiply(
   A=Power(Base=OneMinus(None=@n6), Exp=ScalarParameter() {defaultValue=1.5, parameterName=TopEdgeFade_Power})
   B=ScalarParameter() {defaultValue=1, parameterName=TopEdgeFade_Mult}
  )
  n8 = ScalarParameter() {defaultValue=0.1, parameterName=TopEdgeFade_Add}
  n9 = Multiply(
   A=Multiply(A=Multiply(A=@n4, B=@n5), B=Add(A=@n7, B=@n8))
   B=Multiply(A=ScalarParameter() {defaultValue=1, parameterName=Fade_Mult}, B=ParticleColor().A)
  )
  n10 = Multiply(
   A=TextureCoordinate()
   B=AppendVector(A=ScalarParameter() {defaultValue=1, parameterName=Noise_Tile_X}, B=ScalarParameter() {defaultValue=1, parameterName=Noise_Tile_Y})
  )
  n11 = AppendVector(A=ScalarParameter() {defaultValue=0.1, parameterName=Noise_Pan_X}, B=ScalarParameter() {defaultValue=0.2, parameterName=Noise_Pan_Y})
  n12 = Power(
   Base=TextureSample(UVs=Panner(Coordinate=@n10, Speed=@n11)) {texture=/Game/Characters/Player/Male/Medium/Bodies/M_MED_Viper/Textures/Glitch/T_SoftNoise}.B
   Exp=ScalarParameter() {defaultValue=0.2, parameterName=Noise_Power}
  )
  n13 = ScalarParameter() {defaultValue=2, parameterName=Noise_Mult}
  n14 = Multiply(
   A=Multiply(A=Clamp(None=@n9), B=Multiply(A=@n12, B=@n13))
   B=MaterialFunctionCall(
    Brightness=ScalarParameter() {defaultValue=1, parameterName=Fresnel_Brightness}
    Exponent=ScalarParameter() {defaultValue=4, parameterName=Fresnel_Exponent}
    Vertex / Pixel Normal=VertexNormalWS()
   ) {materialFunction=&MF_Fresnel_FX}.Fresnel Double Sided
  )
  n15 = DepthFade(Opacity=@n14, FadeDistance=ScalarParameter() {defaultValue=50, parameterName=DepthFade})
  NamedRerouteDeclaration(None=ShadingPathSwitch(Default=@n15, Deferred=@n15, Mobile=@n14) {desc=Migrated from FeatureLevelSwitch}) {name=Opacity, nodeColor=<rgba 1 0.5156 0 1>}
  NamedRerouteDeclaration(None=ParticleColor().RGB) {name=Emissive, nodeColor=<rgba 1 0.6797 0 1>}
  NamedRerouteUsage()
  NamedRerouteUsage()
  NamedRerouteUsage()
  PixelNormalWS()
  PixelNormalWS()
 ]
}
