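# Niagara system for the Rift Gate, layering ribbon tendrils, orbiting dots and lights, an energy mesh, light cones, lightning, embers and flashes.
# Contact ribbons follow randomized Bezier splines between cylinder-sampled points, with seeds tied to loop count.
# System.Overload speeds rotation, raises lightning odds, shifts blue-purple toward hot orange-pink and gates bursts around 0.8.
# Embers, flares and flashes share system-level random loop timing and spin to stay aligned.
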NiagaraSystem /Game/Athena/Items/ForagedItems/Rift/FX/RiftGate/NS_RiftGate_Tendrils
dependencies {
 usedDataInterfaces [&NiagaraDataInterfaceCurve, &NiagaraDataInterfaceVectorCurve]
 usedDynamicInputs []
 usedModules [
  &SystemState
  &Timeline
  &self:SystemSpawnScript.SystemScriptSource.SystemScriptGraph.NiagaraNodeAssignment_0.SetVariables_#1
  &self:SystemSpawnScript.SystemScriptSource.SystemScriptGraph.NiagaraNodeAssignment_2.SetVariables_#2
  &self:SystemSpawnScript.SystemScriptSource.SystemScriptGraph.NiagaraNodeAssignment_1.SetVariables_#3
  &self:SystemSpawnScript.SystemScriptSource.SystemScriptGraph.NiagaraNodeAssignment_3.SetVariables_#4
 ]
}
emitters {
 ContactRibbons_Interp {
  emitterUpdateScript [
   &EmitterState(
    Life Cycle Mode=::NewEnumerator1
    Loop Duration=&RandomRangeFloat(Minimum=0.2, Maximum=1, Evaluation Type=::NewEnumerator1)
    Recalculate Duration Each Loop=true
    UseLoopDelay=false
   )
   &SpawnBurst_Instantaneous(Spawn Count=10, Spawn Group=1, Use Spawn Probability=false, UseLoopCountLimit=false)
   &SpawnBurst_Instantaneous(Spawn Count=500, Use Spawn Probability=false, UseLoopCountLimit=false) {enabled=false, module=SpawnBurst_Instantaneous001}
   &SpawnRate(SpawnRate=4, Use Spawn Probability=false) {enabled=false}
  ]
  particleSpawnScript [
   &InitializeParticle(
    Color=<rgba 0.7641 0 1 0.2>
    Lifetime=&RandomRangeFloat(Minimum=2, Maximum=4, Randomness Mode=::NewEnumerator1, Random Seed=$Emitter.LoopCount, Fixed Random Seed=true)
    Ribbon Width=&RandomRangeFloat(Minimum=5, Maximum=20)
    Ribbon Width Mode=::NewEnumerator1
    Sprite Rotation Mode=::NewEnumerator2
    Sprite Size Mode=::NewEnumerator3
    Uniform Sprite Size=&RandomRangeFloat(Minimum=100, Maximum=200)
    UsePositionOffset=false
    Write Material Random=true
   )
   &PartitionParticles(
    Normalized Index In Partition Includes One=true
    Particles Per Partition=10
    Partition Count Mode:ENiagara_InfiniteLoopDuration=::NewEnumerator1
    Partition Mode:ENiagara_PartitionMode=::NewEnumerator1
    Write Partition Index to Ribbon Link Order:ENiagara_AttributeSamplingApplyOutput=::NewEnumerator0
    Write Partition to Ribbon ID=::NewEnumerator0
   )
   &ShapeLocation(
    Apply To Particle Position=::NewEnumerator1
    Cylinder Distribution=::NewEnumerator2
    Cylinder Radius=311.8889
    MaskSpawnGroup=false
    Rotation Mode=::NewEnumerator2
    Shape Primitive=::NewEnumerator1
    U Position=&FloatFromCurve(
     FloatCurve=DataInterface() {propertyValues {Curve {keys [<at 0 0.1>,<at 1 0.2>]}}}
     CurveIndex=&RandomRangeFloat(Minimum=0, Maximum=1, Randomness Mode=::NewEnumerator1, Random Seed=$Emitter.LoopCount, Fixed Random Seed=true)
    )
    UseOwnerScale=true
    Yaw / Pitch / Roll=<0 0.75 0>
   )
   &ShapeLocation(
    Apply To Particle Position=::NewEnumerator1
    Cylinder Distribution=::NewEnumerator2
    Cylinder Radius=420
    MaskSpawnGroup=false
    Rotation Mode=::NewEnumerator2
    Shape Primitive=::NewEnumerator1
    U Position=&FloatFromCurve(
     FloatCurve=DataInterface() {propertyValues {Curve {keys [<at 0 -0.075>,<at 1 0.075>]}}}
     CurveIndex=&RandomRangeFloat(Minimum=0, Maximum=1, Randomness Mode=::NewEnumerator1, Random Seed=$Emitter.LoopCount, Fixed Random Seed=true)
    )
    UseOwnerScale=true
    Yaw / Pitch / Roll=<0 0.75 0>
   ) {module=ShapeLocation001}
   /SetVariables(Particles.StartPosition:Vector3f=$Particles.ShapeLocation001.ShapeVector) {isSetParameters=true}
   /SetVariables(Particles.EndPosition:Vector3f=$Particles.ShapeLocation.ShapeVector) {isSetParameters=true}
   /SetVariables(Particles.MidPoint:Vector3f=&Lerp_Vector(A=$Particles.StartPosition, B=$Particles.EndPosition, Alpha=0.5)) {isSetParameters=true}
   &self:BezierSpline(
    EndHandle:Vector3f=&Add_Vector(
     A=$Particles.MidPoint
     B=&RandomRangeVector(Minimum=<-20 -20 -20>, Maximum=<20 20 20>, Randomness Mode=::NewEnumerator1, Random Seed=$Emitter.LoopCount)
    )
    EndPosition:Vector3f=$Particles.EndPosition
    PositionAlongSpline:NiagaraFloat=$Particles.PartitionParticles.Normalized Index In Partition
    StartHandle:Vector3f=&Add_Vector(
     A=$Particles.MidPoint
     B=&RandomRangeVector(Minimum=<-20 -20 -20>, Maximum=<20 20 20>, Randomness Mode=::NewEnumerator1, Random Seed=$Emitter.LoopCount)
    )
    StartPosition:Vector3f=$Particles.StartPosition
   ) {module=BezierSpline}
  ]
  particleUpdateScript [
   &ParticleState()
   /SetVariables(Particles.Position=&Add_Vector(A=/Niagara/DynamicInputs/Helpers/SimulationPosition(), B=$Particles.Spline_Position)) {isSetParameters=true}
   /SetVariables(
    Particles.Strands_Color:LinearColor=&MultiplyLinearColor(
     Linear Color=&MakeLinearColorFromVectorAndFloat(
      "Vector (RGB)"=&VectorFromFloat(Value=1)
      "Float (Alpha)"=&FloatFromCurve(
       FloatCurve=DataInterface() {propertyValues {Curve {keys [<at 0 0 RCIM_Cubic>, {interpMode=RCIM_Cubic, tangentMode=RCTM_User, time=0.5, value=1}, <at 1 0 RCIM_Cubic>]}}}
       Scale Curve=0.2
      )
     )
     Scale Factor=<1 1 1 1>
    )
    Particles.Strands_RibbonWidth:NiagaraFloat=&FloatFromCurve(
     FloatCurve=DataInterface() {propertyValues {Curve {keys [
         <at 0 0.5 RCIM_Cubic>
         <at 0.2 1>
         {arriveTangent=-0.625, interpMode=RCIM_Cubic, leaveTangent=-0.625, time=0.8, value=1}
         <at 1 0.5 RCIM_Cubic>
        ]}}}
     CurveIndex=$Particles.RibbonLinkOrder
     Scale Curve=&RandomRangeFloat(Minimum=5, Maximum=20)
    )
   ) {isSetParameters=true}
   /SetVariables(
    Particles.AirglowColor:LinearColor=&MultiplyLinearColor(
     Linear Color=&MakeLinearColorFromVectorAndFloat(
      "Vector (RGB)"=&VectorFromFloat(Value=1)
      "Float (Alpha)"=&FloatFromCurve(
       FloatCurve=DataInterface() {propertyValues {Curve {keys [<at 0 0 RCIM_Cubic>, {interpMode=RCIM_Cubic, tangentMode=RCTM_User, time=0.5, value=1}, <at 1 0 RCIM_Cubic>]}}}
       Scale Curve=0.5
      )
     )
     Scale Factor=&MakeFloat4FromVectorAndFloat("Vector (RGB)"=&VectorFromFloat(Value=1), "Float (Alpha)"=1)
    )
    Particles.AirglowRibbonWidth:NiagaraFloat=&Multiply_Float(A=$Particles.RibbonWidth, B=2)
   ) {isSetParameters=true}
   &ScaleColor(
    Scale Alpha=&FloatFromCurve(FloatCurve=DataInterface() {propertyValues {Curve {keys [<at 0 0 RCIM_Cubic>, {interpMode=RCIM_Cubic, tangentMode=RCTM_User, time=0.25, value=1}, <at 1 0 RCIM_Cubic>]}}})
    ScaleA=true
    ScaleRGB=true
   )
   &DynamicMaterialParameters(
    Index 0 Param 1=&FloatFromCurve(FloatCurve=DataInterface() {propertyValues {Curve {keys [<at 0 0>,<at 1 1>]}}}, CurveIndex=$Emitter.Age)
    Index 0 Param 4=&Multiply_Float(A=$Emitter.Age, B=-0.2)
    Index 1 Param 3=&FloatFromCurve(
     FloatCurve=DataInterface() {propertyValues {Curve {keys [<at 0 1>,<at 1 0>]}}}
     CurveIndex=&Frac_Float(Float=&Multiply_Float(A=$Emitter.Age, B=0.2))
    )
    Module.Param0WriteEnabled=true
    Module.Param0WriteEnabled1=true
    Module.Param1WriteEnabled=true
    Module.Param1WriteEnabled1=true
    Module.Param2WriteEnabled=true
    Module.Param2WriteEnabled1=true
    Module.Param3WriteEnabled=true
    Module.Param3WriteEnabled1=true
    Write Parameter Index 1=true
   )
   /SetVariables(Particles.FlarePosition:NiagaraPosition=&Add_Vector(
     A=&SimulationPosition()
     B=&MakeCustomVectorFromBool(
      Bool=&SetBoolByIntComparison(A=0, B=$Particles.PartitionParticles.Partition, Comparison Type=::NewEnumerator4)
      True Vector=$Particles.StartPosition
      False Vector=$Particles.EndPosition
     )
    )) {isSetParameters=true}
   /SetVariables(Particles.FlareColor=&MultiplyLinearColor(Linear Color=<rgba 1 1 1 0.1>, Scale Factor=&MakeVector4_FromLinearColor(LinearColor=$Particles.Color))) {isSetParameters=true}
   &ScaleColor(ScaleA=true, ScaleRGB=true) {module=ScaleColor001}
  ]
  rendererClasses [&NiagaraSpriteRendererProperties, &NiagaraRibbonRendererProperties]
  renderers [
   {rendererClass=&NiagaraSpriteRendererProperties}
   {rendererClass=&NiagaraRibbonRendererProperties, rendererIndex=1}
   {rendererClass=&NiagaraRibbonRendererProperties, rendererIndex=2}
   {rendererClass=&NiagaraRibbonRendererProperties, rendererIndex=3}
   {rendererClass=&NiagaraSpriteRendererProperties, rendererIndex=4}
  ]
 }
 Dots {
  emitterUpdateScript [
   &EmitterState()
   &SpawnRate(SpawnRate=20, Use Spawn Probability=false)
  ]
  particleSpawnScript [
   /SetVariables(Particles.RandomRenderer:NiagaraBool=&RandomBool()) {isSetParameters=true}
   /SetVariables(Particles.DotVisibility:NiagaraInt32=&MakeCustomIntFromBool(Bool=$Particles.RandomRenderer, True INT=1, False INT=0)) {isSetParameters=true}
   /SetVariables(Particles.StarVisibility:NiagaraInt32=&MakeCustomIntFromBool(
     Bool=$Particles.RandomRenderer
     True INT=&RandomRangeInt(Minimum=&RandomRangeInt(Minimum=0, Maximum=1), Maximum=1)
     False INT=1
    )) {isSetParameters=true}
   &InitializeParticle(
    Lifetime=&RandomRangeFloat(Minimum=2, Maximum=3)
    Sprite Size Mode=::NewEnumerator3
    Uniform Sprite Size=&MakeCustomFloatFromBool(
     Bool=$Particles.RandomRenderer
     True Float=&RandomRangeFloat(Minimum=20, Maximum=40)
     False Float=&RandomRangeFloat(Minimum=2, Maximum=4)
    )
    UsePositionOffset=false
    Write Material Random=true
   )
   &ShapeLocation(MaskSpawnGroup=false, Override Seed=false, Sphere Radius=10, UseOwnerScale=true) {module=ShapeLocation001}
  ]
  particleUpdateScript [
   &ParticleState()
   &ScaleSpriteSize(Uniform Curve Sprite Scale=DataInterface() {propertyValues {Curve {keys [
        <at 0 0 RCIM_Cubic>
        {interpMode=RCIM_Cubic, tangentMode=RCTM_User, time=0.2, value=1}
        {interpMode=RCIM_Cubic, tangentMode=RCTM_User, time=0.8, value=1}
        <at 1 0 RCIM_Cubic>
       ]}}})
   /SetVariables(Particles.T:NiagaraFloat=&Add_Float(
     A=&RandomRangeFloat(Minimum=0, Maximum=0.9)
     B=&RandomRangeFloat(
      Minimum=&FloatFromCurve(FloatCurve=DataInterface() {propertyValues {Curve {keys [<at 0 0>,<at 1 0.1>]}}})
      Maximum=&FloatFromCurve(FloatCurve=DataInterface() {propertyValues {Curve {keys [<at 0 0>,<at 1 0.2>]}}})
     )
    )) {isSetParameters=true}
   &UpdateMeshOrientation(
    Facing Direction=$Engine.Owner.SystemXAxis
    Facing Interpolation=false
    Orientation Method=::NewEnumerator1
    Override Quat=false
    Side Direction:Vector3f=$Engine.Owner.SystemYAxis
    Stabilization Mode:ENiagara_OrientVectorCount=::NewEnumerator2
    TransformFacing=false
    TransformSide=false
   )
   &ShapeLocation(
    Additional Quaternion Rotation=true
    Additional Yaw / Pitch / Roll=&MakeVector(
     X=0.25
     Y=&Multiply_Float(
      A=$Emitter.Age
      B=&Lerp_Float(A=&RandomRangeFloat(Minimum=0.01, Maximum=0.05), B=0.5, Alpha=$System.Overload)
     )
     Z=0.25
    )
    Cylinder Distribution=::NewEnumerator2
    Cylinder Height Midpoint=0.6
    Cylinder Radius=&RandomRangeFloat(Minimum=180, Maximum=260)
    MaskSpawnGroup=false
    Offset Coordinate Space=::Local
    Offset Mode=::NewEnumerator0
    Rotation Mode=::NewEnumerator3
    Rotation Quaternion=$Particles.MeshOrientation
    Shape Primitive=::NewEnumerator1
    U Position=&RandomRangeFloat(Minimum=0, Maximum=1)
    UseOwnerScale=true
   )
   /SetVariables(Particles.Position=&Add_Vector(
     A=&Add_Vector(
      A=$Particles.ShapeLocation.ShapeVector
      B=&AddVectorToPosition(Position=$Particles.ShapeLocation001.ShapeVector, Vector=&SimulationPosition())
     )
     B=<0 0 0>
    )) {isSetParameters=true}
   &Color(Color=&LerpLinearColors(
     StartColor=&MakeCustomLinearColorFromBool(Bool=&RandomBool(), True Color=<rgba 2.2923 0 3 1>, False Color=<rgba 0 1.9563 6 1>)
     EndColor=&MakeCustomLinearColorFromBool(Bool=&RandomBool(), True Color=<rgba 5 3.3333 2.5 1>, False Color=<rgba 10 6.6667 5 1>)
     LerpFactor=&FloatFromCurve(FloatCurve=DataInterface() {propertyValues {Curve {keys [<at 0.8 0>,<at 1 1>]}}}, CurveIndex=$System.Overload)
    ))
   &SolveForcesAndVelocity(Clamp Velocity=false, Limit Acceleration=false)
  ]
  rendererClasses [&NiagaraSpriteRendererProperties]
  renderers [
   {rendererClass=&NiagaraSpriteRendererProperties}
   {rendererClass=&NiagaraSpriteRendererProperties, rendererIndex=1}
  ]
 }
 Embers_Flare {
  emitterUpdateScript [
   &EmitterState(Life Cycle Mode=::NewEnumerator1, Loop Duration=$System.Random_LoopDuration, Recalculate Duration Each Loop=true, UseLoopDelay=false) {module=EmitterState001}
   &SpawnBurst_Instantaneous(
    Spawn Probability=&MakeCustomFloatFromBool(Bool=&SetBoolByFloatComparison(A=$System.Overload, B=0.8), True Float=1, False Float=0)
    Use Spawn Probability=true
    UseLoopCountLimit=false
   )
   /SetVariables(Emitter.RandomSpin:NiagaraFloat=$System.RandomSpin) {isSetParameters=true}
  ]
  particleSpawnScript [
   &InitializeParticle(
    Color=<rgba 10 3.3333 2 1>
    Lifetime=0.2
    Sprite Rotation Mode=::NewEnumerator2
    Sprite Size Mode=::NewEnumerator3
    Uniform Sprite Size=200
    UsePositionOffset=false
    Write Material Random=true
   )
   &CameraOffset(Camera Offset Amount=50)
   &ShapeLocation(
    Cylinder Distribution=::NewEnumerator2
    Cylinder Radius=280
    MaskSpawnGroup=false
    Rotation Angle:NiagaraFloat=$Emitter.RandomSpin
    Rotation Mode=::NewEnumerator1
    Shape Origin=$Engine.Emitter.SimulationPosition
    Shape Primitive=::NewEnumerator1
    U Position=&RandomRangeFloat(Minimum=0, Maximum=0)
    UseOwnerScale=true
   ) {module=ShapeLocation001}
  ]
  particleUpdateScript [
   &ParticleState()
   &ScaleColor(
    Scale Alpha=&FloatFromCurve(FloatCurve=DataInterface() {propertyValues {Curve {keys [<at 0 0>,<at 0.1 1>,<at 1 0>]}}})
    ScaleA=true
    ScaleRGB=true
   )
   &SolveForcesAndVelocity(Clamp Velocity=false, Limit Acceleration=false)
  ]
  rendererClasses [&NiagaraSpriteRendererProperties]
  renderers [{rendererClass=&NiagaraSpriteRendererProperties}]
 }
 Embers_Trapped {
  emitterUpdateScript [
   &EmitterState(Life Cycle Mode=::NewEnumerator1, Loop Duration=$System.Random_LoopDuration, Recalculate Duration Each Loop=true, UseLoopDelay=false)
   &SpawnRate(SpawnRate=100, Use Spawn Probability=false) {enabled=false}
   &SpawnBurst_Instantaneous(
    Spawn Count=&RandomRangeInt(Minimum=30, Maximum=100)
    Spawn Probability=&MakeCustomFloatFromBool(Bool=&SetBoolByFloatComparison(A=$System.Overload, B=0.8), True Float=1, False Float=0)
    Use Spawn Probability=true
    UseLoopCountLimit=false
   )
   /SetVariables(Emitter.Random_SpinRange:NiagaraFloat=$System.RandomSpin) {isSetParameters=true}
  ]
  particleSpawnScript [
   &InitializeParticle(
    Color=<rgba 10 3.5895 1 1>
    Lifetime=&RandomRangeFloat(Minimum=1, Maximum=4)
    Sprite Size Mode=::NewEnumerator3
    Uniform Sprite Size=&RandomRangeFloat(Minimum=3, Maximum=6)
    UsePositionOffset=false
    Write Material Random=true
   )
   /SetVariables(Particles.SpinRange:NiagaraFloat=$Emitter.Random_SpinRange) {isSetParameters=true}
   &ShapeLocation(
    Cylinder Distribution=::NewEnumerator2
    Cylinder Radius=280
    MaskSpawnGroup=false
    Rotation Angle:NiagaraFloat=$Particles.SpinRange
    Rotation Mode=::NewEnumerator1
    Shape Origin=$Engine.Emitter.SimulationPosition
    Shape Primitive=::NewEnumerator1
    U Position=&Add_Float(A=0, B=0)
    UseOwnerScale=true
   ) {module=ShapeLocation001}
   &AddVelocity(
    Cone Angle=&RandomRangeFloat(Minimum=20, Maximum=40, Randomness Mode=::NewEnumerator1, Random Seed=$Emitter.LoopCount, Fixed Random Seed=true)
    Cone Axis=&Add_Vector(
     A=&Add_Vector(
      A=&Multiply_VectorByFloat(Vector=$Particles.ShapeLocation001.ShapeNormal, Float=-1)
      B=&Multiply_VectorByFloat(Vector=$Particles.ShapeLocation001.ShapeTangent, Float=1)
     )
     B=&RandomRangeVector(Minimum=<0 0 0>, Maximum=<0 0 0>)
    )
    Module.UseConeFalloff=false
    Module.UseVelDistribution=false
    Override Seed=false
    Rotation Coordinate Space=::World
    Velocity Mode=::NewEnumerator2
    Velocity Speed=&RandomRangeFloat(Minimum=300, Maximum=800)
   )
  ]
  particleUpdateScript [
   &ParticleState()
   &Drag(Drag=1.5, Use Linear Drag=true, Use Rotational Drag=true)
   &ScaleSpriteSize(Uniform Curve Sprite Scale=DataInterface() {propertyValues {Curve {keys [{tangentMode=RCTM_User, time=0, value=1}, {tangentMode=RCTM_User, time=1, value=0}]}}})
   &ShapeLocation(
    Apply To Particle Position=::NewEnumerator1
    Cylinder Distribution=::NewEnumerator2
    Cylinder Height=50
    Cylinder Radius=&Lerp_Float(
     A=&RandomRangeFloat(Minimum=230, Maximum=270)
     B=&RandomRangeFloat(Minimum=160, Maximum=220)
     Alpha=$Particles.NormalizedAge
    )
    MaskSpawnGroup=false
    Rotation Angle:NiagaraFloat=&Frac_Float(Float=&Add_Float(A=$Particles.SpinRange, B=&Multiply_Float(A=$Particles.Age, B=0.25)))
    Rotation Mode=::NewEnumerator1
    Shape Primitive=::NewEnumerator1
    U Position=&FloatFromCurve(
     FloatCurve=DataInterface() {propertyValues {Curve {keys [<at 0 0>,<at 1 0.15>]}}}
     CurveIndex=&RandomRangeFloat(Minimum=0, Maximum=1, Randomness Mode=::NewEnumerator1, Random Seed=$Emitter.LoopCount)
    )
    UseOwnerScale=true
    V Position=&RandomRangeFloat(Minimum=0, Maximum=1)
   )
   &SolveForcesAndVelocity(Clamp Velocity=false, Limit Acceleration=false)
   &ScaleSpriteSizeBySpeed(Max Scale Factor=&MakeVector2D(
     X=1
     Y=&FloatFromCurve(FloatCurve=DataInterface() {propertyValues {Curve {keys [<at 0 5>,<at 0.3 1>]}}})
    ))
   /SetVariables(Particles.Position=&Lerp_Vector(
     A=Linked() {linkedVariable {name=Particles.Position, type=NiagaraPosition}}
     B=&Add_Vector(A=&SimulationPosition(), B=$Particles.ShapeLocation.ShapeVector)
     Alpha=&FloatFromCurve(
      FloatCurve=DataInterface() {propertyValues {Curve {keys [<at 0 0 RCIM_Cubic>, {interpMode=RCIM_Cubic, tangentMode=RCTM_User, time=2, value=1}]}}}
      CurveIndex=$Particles.Age
     )
    )) {isSetParameters=true}
  ]
  rendererClasses [&NiagaraSpriteRendererProperties]
  renderers [{rendererClass=&NiagaraSpriteRendererProperties}]
 }
 EnergyMesh001 {
  emitterUpdateScript [
   &EmitterState(Life Cycle Mode=::NewEnumerator1, Loop Behavior=::NewEnumerator1, Loop Duration Mode=::NewEnumerator1, UseLoopDelay=false)
   &SpawnBurst_Instantaneous(Use Spawn Probability=false, UseLoopCountLimit=false)
  ]
  particleSpawnScript [
   &InitializeParticle(Color=<rgba 0 0.3261 1 1>, Mesh Scale Mode=::NewEnumerator3, Mesh Uniform Scale=1.3095, UsePositionOffset=false, Write Material Random=true)
   &InitialMeshOrientation(Mesh Axis To Orient=<0 0 1>, Mesh Orientation Mode=::NewEnumerator2, Use Rotation Vector=false)
  ]
  particleUpdateScript [
   &ParticleState(Kill Particles When Lifetime Has Elapsed=false, Let Infinitely Lived Particles Die When Emitter Deactivates=true)
   &UpdateMeshOrientation(Rotation Rate=&Lerp_Float(A=0, B=0.2, Alpha=$System.Overload), Rotation Vector=<0 0 1>)
   &Color(Color=&MakeCustomLinearColorFromBool(
     Bool=&SetBoolByFloatComparison(A=$System.Overload, B=0.8, Comparison Type=::NewEnumerator7)
     True Color=&LerpLinearColors(StartColor=<rgba 0 0 0 0.1>, EndColor=<rgba 0 0 0 0.2>, LerpFactor=$System.Overload)
     False Color=<rgba 0 0 0 0.5>
    ))
   &DynamicMaterialParameters(
    Index 0 Param 2=&FloatFromCurve(FloatCurve=DataInterface() {propertyValues {Curve {keys [<at 0 1.5>,<at 1 1>]}}}, CurveIndex=$Particles.Age)
    Index 0 Param 3=&FloatFromCurve(FloatCurve=DataInterface() {propertyValues {Curve {keys [<at 0.8 0>,<at 1 1>]}}}, CurveIndex=$System.Overload)
    Module.Param0WriteEnabled=true
    Module.Param1WriteEnabled=true
    Module.Param2WriteEnabled=true
    Module.Param3WriteEnabled=true
   )
  ]
  rendererClasses [&NiagaraMeshRendererProperties]
  renderers [{rendererClass=&NiagaraMeshRendererProperties, rendererIndex=0}]
 }
 Flash {
  emitterUpdateScript [
   &EmitterState(Life Cycle Mode=::NewEnumerator1, Loop Duration=$System.Random_LoopDuration, Recalculate Duration Each Loop=true, UseLoopDelay=false) {module=EmitterState001}
   &SpawnBurst_Instantaneous(
    Spawn Probability=&MakeCustomFloatFromBool(Bool=&SetBoolByFloatComparison(A=$System.Overload, B=0.8), True Float=1, False Float=0)
    Use Spawn Probability=true
    UseLoopCountLimit=false
   )
   /SetVariables(Emitter.RandomSpin:NiagaraFloat=$System.RandomSpin) {isSetParameters=true}
  ]
  particleSpawnScript [
   &InitializeParticle(
    Color=<rgba 10 3.3333 2 1>
    Lifetime=0.2
    Mesh Scale Mode=::NewEnumerator3
    Mesh Uniform Scale=1.5
    Sprite Rotation Mode=::NewEnumerator2
    Sprite Size Mode=::NewEnumerator3
    Uniform Sprite Size=200
    UsePositionOffset=false
    Write Material Random=true
   )
   /SetVariables(Particles.MeshOrientation=&MakeQuatFromAxisAngle(
     Axis=&RandomRangeVector(Minimum=<0 0 0>, Maximum=<1 1 1>, Evaluation Type=::NewEnumerator1)
     AngleInDegrees=90
    )) {isSetParameters=true}
   &CameraOffset(Camera Offset Amount=50)
   &ShapeLocation(
    Cylinder Distribution=::NewEnumerator2
    Cylinder Radius=280
    MaskSpawnGroup=false
    Rotation Angle:NiagaraFloat=$Emitter.RandomSpin
    Rotation Mode=::NewEnumerator1
    Shape Origin=$Engine.Emitter.SimulationPosition
    Shape Primitive=::NewEnumerator1
    U Position=&RandomRangeFloat(Minimum=0, Maximum=0)
    UseOwnerScale=true
   ) {module=ShapeLocation001}
  ]
  particleUpdateScript [
   &ParticleState()
   &DynamicMaterialParameters(
    Index 0 Param 1=&FloatFromCurve(FloatCurve=DataInterface() {propertyValues {Curve {keys [{arriveTangent=1, time=0.56, value=0}, <at 1 1>]}}})
    Index 0 Param 2=45
    Index 0 Param 3=1
    Index 0 Param 4=1
    Module.Param0WriteEnabled=true
    Module.Param1WriteEnabled=true
    Module.Param2WriteEnabled=true
    Module.Param3WriteEnabled=true
   )
   &ScaleColor(
    Scale Alpha=&FloatFromCurve(FloatCurve=DataInterface() {propertyValues {Curve {keys [<at 0 1>,<at 1 0>]}}})
    Scale RGB=&VectorFromCurve(VectorCurve=DataInterface() {propertyValues {XCurve {keys [<at 0 1>,<at 1 1>]}, YCurve {keys [<at 0 1>,<at 1 1>]}, ZCurve {keys [<at 0 1>,<at 1 1>]}}})
    ScaleA=true
    ScaleRGB=true
   )
   &ScaleMeshSize(Scale Factor=&VectorFromFloat(Value=&FloatFromCurve(FloatCurve=DataInterface() {propertyValues {Curve {keys [
          {arriveTangent=0.3871, interpMode=RCIM_Cubic, time=0, value=0.5}
          {arriveTangent=0.6012, interpMode=RCIM_Cubic, leaveTangent=0.6012, tangentMode=RCTM_User, time=0.7128, value=0.698}
          {arriveTangent=1.7425, interpMode=RCIM_Cubic, leaveTangent=1.7425, tangentMode=RCTM_User, time=1, value=1}
         ]}}})))
  ]
  rendererClasses [&NiagaraMeshRendererProperties]
  renderers [{rendererClass=&NiagaraMeshRendererProperties, rendererIndex=0}]
 }
 LightCone_mesh {
  emitterUpdateScript [
   &EmitterState(
    Inactive Response=::NewEnumerator1
    Life Cycle Mode=::NewEnumerator1
    Loop Behavior=::NewEnumerator1
    Loop Duration Mode=::NewEnumerator1
    UseLoopDelay=false
   )
   &SpawnBurst_Instantaneous(Spawn Count=2, Use Spawn Probability=false)
  ]
  particleSpawnScript [
   &PartitionParticles()
   &InitializeParticle(
    Mesh Scale=&MakeCustomVectorFromBool(
     Bool=&SetBoolByIntComparison(A=$Particles.PartitionParticles.Partition, B=0, Comparison Type=::NewEnumerator0)
     True Vector=<1 1 1>
     False Vector=<1 1 -1>
    )
    Mesh Scale Mode=::NewEnumerator4
    Sprite Size Mode=::NewEnumerator3
    Uniform Sprite Size=400
    UsePositionOffset=false
    Write Material Random=true
   ) {module=InitializeParticle001}
   &InitialMeshOrientation(Mesh Axis To Orient=<0 0 -1>, Mesh Orientation Mode=::NewEnumerator2, Use Rotation Vector=false)
  ]
  particleUpdateScript [
   &ParticleState(Kill Particles When Lifetime Has Elapsed=false, Let Infinitely Lived Particles Die When Emitter Deactivates=true)
   &Color(Color=&LerpLinearColors(
     StartColor=<rgba 0 0.3261 1 0.2>
     EndColor=&FNTimeOfDayColor(
      Night Color=<rgba 20 4.5333 4 0.2>
      Morning Color=<rgba 20 4.5333 4 0.5>
      Day Color=<rgba 20 4.5333 4 0.5>
      Evening Color=<rgba 20 4.5333 4 0.5>
     )
     LerpFactor=$System.Overload
    ))
  ]
  rendererClasses [&NiagaraMeshRendererProperties]
  renderers [{rendererClass=&NiagaraMeshRendererProperties, rendererIndex=0}]
 }
 LightningMesh {
  emitterUpdateScript [
   &EmitterState(
    Life Cycle Mode=::NewEnumerator1
    Loop Delay=&RandomRangeFloat(Minimum=0, Maximum=0)
    Loop Duration=&RandomRangeFloat(Minimum=0.5, Maximum=1)
    UseLoopDelay=false
   )
   &SpawnRate(
    Spawn Probability=&FloatFromCurve(
     FloatCurve=DataInterface() {propertyValues {Curve {keys [<at 0.5 0 RCIM_Cubic>, {arriveTangent=2.6127, interpMode=RCIM_Cubic, leaveTangent=2.6127, tangentMode=RCTM_User, time=1, value=1}]}}}
     CurveIndex=$System.Overload
     Scale Curve=&RandomRangeFloat(Minimum=0, Maximum=1)
    )
    SpawnRate=&RandomRangeFloat(Minimum=2, Maximum=3)
    Use Spawn Probability=true
   )
   /SetVariables(Emitter.RandomPosition:NiagaraFloat=&RandomRangeFloat(Minimum=0, Maximum=1, Randomness Mode=::NewEnumerator1, Random Seed=$Emitter.LoopCount, Fixed Random Seed=true)) {isSetParameters=true, module=SetVariables_#5}
  ]
  particleSpawnScript [
   &InitializeParticle(
    Color=&MultiplyLinearColor(Linear Color=<rgba 1 1 1 1>, Scale Factor=&MakeVector4_FromFloat(Float=1.15))
    Lifetime=&FloatFromCurve(FloatCurve=DataInterface() {propertyValues {Curve {keys [<at 0 0.25 RCIM_Cubic>, {arriveTangent=0.4797, interpMode=RCIM_Cubic, leaveTangent=0.4797, tangentMode=RCTM_User, time=1, value=0.4}]}}})
    Mesh Scale=&Multiply_VectorByFloat(
     Vector=&MaskVectorBySpawnGroup(Masked Vector=<1 1 1>, Passthrough Vector=<1 1 -1>)
     Float=&RandomRangeFloat(Minimum=5, Maximum=9)
    )
    Mesh Scale Mode=::NewEnumerator4
    Sprite Size Mode=::NewEnumerator3
    Uniform Sprite Size=100
    UsePositionOffset=false
    Write Material Random=true
   )
   &InitialMeshOrientation(Use Rotation Vector=true)
   /SetVariables(
    Particles.LightningSubImageIndex:NiagaraFloat=&UniformRangedFloat(Minimum=0, Maximum=3)
    Particles.SpriteSize=&Multiply_Vector2D_ByFloat(
     Float=&MaskFloatBySpawnGroup(Masked Float=&RandomRangeFloat(Minimum=1, Maximum=2), Passthrough Float=1, Spawn Group Mask=1)
     Vector2D=$Particles.SpriteSize
    )
   ) {isSetParameters=true, module=SetVariables_#6}
  ]
  particleUpdateScript [
   &ParticleState()
   &ShapeLocation(
    Cylinder Distribution=::NewEnumerator2
    Cylinder Radius=300
    MaskSpawnGroup=false
    Rotation Mode=::NewEnumerator2
    Shape Primitive=::NewEnumerator1
    U Position=$Emitter.RandomPosition
    UseOwnerScale=true
    Yaw / Pitch / Roll=<0 0.25 0>
   )
   &ScaleMeshSize(Scale Factor=&VectorFromFloat(Value=1.5))
   &ScaleColor(
    Scale Alpha=&FloatFromCurve(FloatCurve=DataInterface() {propertyValues {Curve {keys [<at 0.5 1 RCIM_Cubic>,<at 1 0 RCIM_Cubic>]}}})
    Scale RGB=&Multiply_Vector(
     A=&VectorFromFloat(Value=&FloatFromCurve(FloatCurve=DataInterface() {propertyValues {Curve {keys [
           <at 0 0 RCIM_Cubic>
           {arriveTangent=15, interpMode=RCIM_Cubic, leaveTangent=15, time=0.1, value=20}
           {arriveTangent=-22.2222, interpMode=RCIM_Cubic, leaveTangent=-22.2222, time=0.2, value=3}
           <at 1 0 RCIM_Cubic>
          ]}}}))
     B=&Lerp_Vector(
      A=&MakeVectorFromLinearColorRGB(Color=<rgba 0 0.3261 1 1>)
      B=&MakeVectorFromLinearColorRGB(Color=<rgba 1 0.2267 0.2 1>)
      Alpha=&FloatFromCurve(FloatCurve=DataInterface() {propertyValues {Curve {keys [<at 0.8 0>,<at 1 1>]}}}, CurveIndex=$System.Overload)
     )
    )
    ScaleA=true
    ScaleRGB=true
   )
   &CameraOffset(Camera Offset Amount=10)
   &DynamicMaterialParameters(
    Index 0 Param 1=&FloatFromCurve(FloatCurve=DataInterface() {propertyValues {Curve {keys [<at 0 1>,<at 1 0>]}}})
    Index 0 Param 2=&FloatFromCurve(
     FloatCurve=DataInterface() {propertyValues {Curve {keys [
         <at 0 1 RCIM_Cubic>
         {arriveTangent=0.987, interpMode=RCIM_Cubic, leaveTangent=0.987, tangentMode=RCTM_User, time=0.4992, value=1}
         <at 0.571 1 RCIM_Cubic>
         <at 1 1>
        ]}}}
     Scale Curve=5000
    )
    Index 0 Param 3=&RandomRangeFloat(Minimum=1, Maximum=2)
    Index 0 Param 4=&RandomRangeFloat(Minimum=0, Maximum=1)
    Module.Param0WriteEnabled=true
    Module.Param1WriteEnabled=true
    Module.Param2WriteEnabled=true
    Module.Param3WriteEnabled=true
   ) {module=DynamicMaterialParameters004}
   /SetVariables(Particles.LightColor=&LerpLinearColors(
     StartColor=<rgba 1 2.3042 5 1>
     EndColor=<rgba 5 1.6667 1 1>
     LerpFactor=&FloatFromCurve(FloatCurve=DataInterface() {propertyValues {Curve {keys [<at 0.8 0>,<at 1 1>]}}}, CurveIndex=$System.Overload)
    )) {isSetParameters=true, module=SetVariables_#7}
   &FN_Light_Attributes(
    Color=$Particles.LightColor
    Create Color Attribute=true
    Create Exponent Attribute=true
    Create Position Attribute=true
    Create Radius Attribute=true
    Exponent=2
    Position=&Add_Vector(
     A=Linked() {linkedVariable {name=Particles.Position, type=NiagaraPosition}}
     B=&Multiply_VectorByFloat(Vector=$Particles.ShapeLocation.ShapeNormal, Float=100)
    )
    Radius=200
    Scale Color=&FloatFromCurve(
     FloatCurve=DataInterface() {propertyValues {Curve {keys [
         <at 0 0 RCIM_Cubic>
         {arriveTangent=0.875, interpMode=RCIM_Cubic, leaveTangent=0.875, time=0.1, value=0.9}
         {arriveTangent=-0.2131, interpMode=RCIM_Cubic, leaveTangent=-0.2131, tangentMode=RCTM_User, time=0.4, value=0.35}
         {arriveTangent=-0.5833, interpMode=RCIM_Cubic, leaveTangent=-0.5833, time=0.7, value=0.5}
         <at 1 0 RCIM_Cubic>
        ]}}}
     Scale Curve=1.5
    )
   )
   &ScaleColor(
    Scale RGB=&Lerp_Vector(
     A=&MakeVectorFromLinearColorRGB(Color=<rgba 0 0.3261 1 1>)
     B=&MakeVectorFromLinearColorRGB(Color=<rgba 1 0.2267 0.2 1>)
     Alpha=&FloatFromCurve(FloatCurve=DataInterface() {propertyValues {Curve {keys [<at 0.8 0>,<at 1 1>]}}}, CurveIndex=$System.Overload)
    )
    ScaleA=true
    ScaleRGB=true
   ) {enabled=false, module=ScaleColor001}
  ]
  rendererClasses [
   &NiagaraMeshRendererProperties
   &NiagaraSpriteRendererProperties
   &NiagaraLightRendererProperties
  ]
  renderers [
   {rendererClass=&NiagaraMeshRendererProperties, rendererIndex=0}
   {rendererClass=&NiagaraSpriteRendererProperties, rendererIndex=1}
   {rendererClass=&NiagaraLightRendererProperties, rendererIndex=2}
   {rendererClass=&NiagaraSpriteRendererProperties, rendererIndex=3}
  ]
 }
 LightningMesh001 {
  emitterUpdateScript [
   &EmitterState(
    Life Cycle Mode=::NewEnumerator1
    Loop Delay=&RandomRangeFloat(Minimum=0, Maximum=0)
    Loop Duration=&RandomRangeFloat(Minimum=0.5, Maximum=0.2)
    UseLoopDelay=false
   )
   &SpawnRate(
    Spawn Probability=&FloatFromCurve(
     FloatCurve=DataInterface() {propertyValues {Curve {keys [<at 0.5 0 RCIM_Cubic>, {arriveTangent=2.6127, interpMode=RCIM_Cubic, leaveTangent=2.6127, tangentMode=RCTM_User, time=1, value=1}]}}}
     CurveIndex=$System.Overload
     Scale Curve=&RandomRangeFloat(Minimum=0, Maximum=1)
    )
    SpawnRate=&RandomRangeFloat(Minimum=2, Maximum=3)
    Use Spawn Probability=true
   )
   /SetVariables(Emitter.RandomPosition:NiagaraFloat=&RandomRangeFloat(Minimum=0, Maximum=1, Randomness Mode=::NewEnumerator1, Random Seed=$Emitter.LoopCount, Fixed Random Seed=true)) {isSetParameters=true, module=SetVariables_#5}
  ]
  particleSpawnScript [
   &InitializeParticle(
    Color=&MultiplyLinearColor(Linear Color=<rgba 1 1 1 1>, Scale Factor=&MakeVector4_FromFloat(Float=1.15))
    Lifetime=&FloatFromCurve(FloatCurve=DataInterface() {propertyValues {Curve {keys [<at 0 0.25 RCIM_Cubic>, {arriveTangent=0.4797, interpMode=RCIM_Cubic, leaveTangent=0.4797, tangentMode=RCTM_User, time=1, value=0.4}]}}})
    Mesh Scale=&Multiply_VectorByFloat(
     Vector=&MaskVectorBySpawnGroup(Masked Vector=<1 1 1>, Passthrough Vector=<1 1 -1>)
     Float=&RandomRangeFloat(Minimum=5, Maximum=9)
    )
    Mesh Scale Mode=::NewEnumerator4
    Sprite Size Mode=::NewEnumerator3
    Uniform Sprite Size=100
    UsePositionOffset=false
    Write Material Random=true
   )
   &InitialMeshOrientation(Use Rotation Vector=true)
   /SetVariables(
    Particles.LightningSubImageIndex:NiagaraFloat=&UniformRangedFloat(Minimum=0, Maximum=3)
    Particles.SpriteSize=&Multiply_Vector2D_ByFloat(
     Float=&MaskFloatBySpawnGroup(Masked Float=&RandomRangeFloat(Minimum=1, Maximum=2), Passthrough Float=1, Spawn Group Mask=1)
     Vector2D=$Particles.SpriteSize
    )
   ) {isSetParameters=true, module=SetVariables_#6}
  ]
  particleUpdateScript [
   &ParticleState()
   &ShapeLocation(
    Cylinder Distribution=::NewEnumerator2
    Cylinder Radius=300
    MaskSpawnGroup=false
    Rotation Mode=::NewEnumerator2
    Shape Primitive=::NewEnumerator1
    U Position=$Emitter.RandomPosition
    UseOwnerScale=true
    Yaw / Pitch / Roll=<0 0.25 0>
   )
   &ScaleMeshSize(Scale Factor=&VectorFromFloat(Value=1.5))
   &ScaleColor(
    Scale Alpha=&FloatFromCurve(FloatCurve=DataInterface() {propertyValues {Curve {keys [<at 0.5 1 RCIM_Cubic>,<at 1 0 RCIM_Cubic>]}}})
    Scale RGB=&Multiply_Vector(
     A=&VectorFromFloat(Value=&FloatFromCurve(FloatCurve=DataInterface() {propertyValues {Curve {keys [
           <at 0 0 RCIM_Cubic>
           {arriveTangent=15, interpMode=RCIM_Cubic, leaveTangent=15, time=0.1, value=20}
           {arriveTangent=-22.2222, interpMode=RCIM_Cubic, leaveTangent=-22.2222, time=0.2, value=3}
           <at 1 0 RCIM_Cubic>
          ]}}}))
     B=&Lerp_Vector(
      A=&MakeVectorFromLinearColorRGB(Color=<rgba 0 0.3261 1 1>)
      B=&MakeVectorFromLinearColorRGB(Color=<rgba 1 0.2267 0.2 1>)
      Alpha=&FloatFromCurve(FloatCurve=DataInterface() {propertyValues {Curve {keys [<at 0.8 0>,<at 1 1>]}}}, CurveIndex=$System.Overload)
     )
    )
    ScaleA=true
    ScaleRGB=true
   )
   &CameraOffset(Camera Offset Amount=10)
   &DynamicMaterialParameters(
    Index 0 Param 1=&FloatFromCurve(FloatCurve=DataInterface() {propertyValues {Curve {keys [<at 0 1>,<at 1 0>]}}})
    Index 0 Param 2=&FloatFromCurve(
     FloatCurve=DataInterface() {propertyValues {Curve {keys [
         <at 0 1 RCIM_Cubic>
         {arriveTangent=0.987, interpMode=RCIM_Cubic, leaveTangent=0.987, tangentMode=RCTM_User, time=0.4992, value=1}
         <at 0.571 1 RCIM_Cubic>
         <at 1 1>
        ]}}}
     Scale Curve=5000
    )
    Index 0 Param 3=&RandomRangeFloat(Minimum=1, Maximum=2)
    Index 0 Param 4=&RandomRangeFloat(Minimum=0, Maximum=1)
    Module.Param0WriteEnabled=true
    Module.Param1WriteEnabled=true
    Module.Param2WriteEnabled=true
    Module.Param3WriteEnabled=true
   ) {module=DynamicMaterialParameters004}
   /SetVariables(Particles.LightColor=&LerpLinearColors(
     StartColor=<rgba 1 2.3042 5 1>
     EndColor=<rgba 5 1.6667 1 1>
     LerpFactor=&FloatFromCurve(FloatCurve=DataInterface() {propertyValues {Curve {keys [<at 0.8 0>,<at 1 1>]}}}, CurveIndex=$System.Overload)
    )) {isSetParameters=true, module=SetVariables_#7}
   &FN_Light_Attributes(
    Color=$Particles.LightColor
    Create Color Attribute=true
    Create Exponent Attribute=true
    Create Position Attribute=true
    Create Radius Attribute=true
    Exponent=2
    Position=&Add_Vector(
     A=Linked() {linkedVariable {name=Particles.Position, type=NiagaraPosition}}
     B=&Multiply_VectorByFloat(Vector=$Particles.ShapeLocation.ShapeNormal, Float=100)
    )
    Radius=200
    Scale Color=&FloatFromCurve(
     FloatCurve=DataInterface() {propertyValues {Curve {keys [
         <at 0 0 RCIM_Cubic>
         {arriveTangent=0.875, interpMode=RCIM_Cubic, leaveTangent=0.875, time=0.1, value=0.9}
         {arriveTangent=-0.2131, interpMode=RCIM_Cubic, leaveTangent=-0.2131, tangentMode=RCTM_User, time=0.4, value=0.35}
         {arriveTangent=-0.5833, interpMode=RCIM_Cubic, leaveTangent=-0.5833, time=0.7, value=0.5}
         <at 1 0 RCIM_Cubic>
        ]}}}
     Scale Curve=1.5
    )
   )
   &ScaleColor(
    Scale RGB=&Lerp_Vector(
     A=&MakeVectorFromLinearColorRGB(Color=<rgba 0 0.3261 1 1>)
     B=&MakeVectorFromLinearColorRGB(Color=<rgba 1 0.2267 0.2 1>)
     Alpha=&FloatFromCurve(FloatCurve=DataInterface() {propertyValues {Curve {keys [<at 0.8 0>,<at 1 1>]}}}, CurveIndex=$System.Overload)
    )
    ScaleA=true
    ScaleRGB=true
   ) {enabled=false, module=ScaleColor001}
  ]
  rendererClasses [
   &NiagaraMeshRendererProperties
   &NiagaraSpriteRendererProperties
   &NiagaraLightRendererProperties
  ]
  renderers [
   {rendererClass=&NiagaraMeshRendererProperties, rendererIndex=0}
   {rendererClass=&NiagaraSpriteRendererProperties, rendererIndex=1}
   {rendererClass=&NiagaraLightRendererProperties, rendererIndex=2}
   {rendererClass=&NiagaraSpriteRendererProperties, rendererIndex=3}
  ]
 }
 Lights {
  emitterUpdateScript [
   &EmitterState(Life Cycle Mode=::NewEnumerator1, Loop Behavior=::NewEnumerator1, Loop Duration Mode=::NewEnumerator1, UseLoopDelay=false)
   &SpawnBurst_Instantaneous(Spawn Count=2, Use Spawn Probability=false, UseLoopCountLimit=false)
  ]
  particleSpawnScript [
   &PartitionParticles()
   &InitializeParticle(
    Color=&MakeCustomLinearColorFromBool(
     Bool=&SetBoolByIntComparison(A=$Particles.PartitionParticles.Partition, B=0, Comparison Type=::NewEnumerator0)
     True Color=<rgba 4.5845 0 6 1>
     False Color=<rgba 0 3.9126 12 1>
    )
    Lifetime=2
    Sprite Size Mode=::NewEnumerator3
    UsePositionOffset=false
    Write Material Random=true
   )
  ]
  particleUpdateScript [
   &ParticleState(Kill Particles When Lifetime Has Elapsed=false, Let Infinitely Lived Particles Die When Emitter Deactivates=true)
   &Color(Color=&LerpLinearColors(
     StartColor=&MakeCustomLinearColorFromBool(
      Bool=&SetBoolByIntComparison(A=$Particles.PartitionParticles.Partition, B=0, Comparison Type=::NewEnumerator0)
      True Color=<rgba 0 3.9126 12 1>
      False Color=<rgba 4.5845 0 6 1>
     )
     EndColor=&MakeCustomLinearColorFromBool(
      Bool=&SetBoolByIntComparison(A=$Particles.PartitionParticles.Partition, B=0, Comparison Type=::NewEnumerator0)
      True Color=<rgba 12 3.6 3.88 1>
      False Color=<rgba 12 3.6 5 1>
     )
     LerpFactor=&FloatFromCurve(FloatCurve=DataInterface() {propertyValues {Curve {keys [<at 0.8 0>,<at 1 1>]}}}, CurveIndex=$System.Overload)
    ))
   /SetVariables(Particles.Position=&RotateVectorByQuaternion(
     Vector To Rotate=&MakeCustomVectorFromBool(
      Bool=&SetBoolByIntComparison(A=$Particles.PartitionParticles.Partition, B=0, Comparison Type=::NewEnumerator0)
      True Vector=<0 100 1>
      False Vector=<0 -100 0>
     )
     Quaternion=&MakeQuatFromAxisAngle(
      Axis=<1 0 0>
      AngleInDegrees=&Multiply_Float(A=$Emitter.Age, B=&Lerp_Float(A=90, B=360, Alpha=$System.Overload))
     )
    )) {isSetParameters=true}
   &FN_Light_Attributes(
    Create Exponent Attribute=true
    Create Radius Attribute=true
    Exponent=2
    Radius=&Lerp_Float(A=500, B=500, Alpha=$System.Overload)
   )
  ]
  rendererClasses [&NiagaraSpriteRendererProperties, &NiagaraLightRendererProperties]
  renderers [
   {rendererClass=&NiagaraSpriteRendererProperties}
   {rendererClass=&NiagaraLightRendererProperties, rendererIndex=1}
  ]
 }
 Tesla {
  emitterUpdateScript [
   &EmitterState(
    Life Cycle Mode=::NewEnumerator1
    Loop Duration=&RandomRangeFloat(Minimum=0.1, Maximum=2)
    Recalculate Duration Each Loop=true
    UseLoopDelay=false
   )
   &SpawnBurst_Instantaneous(
    Spawn Count=10
    Spawn Probability=&MakeCustomFloatFromBool(Bool=&SetBoolByFloatComparison(A=$System.Overload, B=0.8), True Float=1, False Float=0)
    Use Spawn Probability=true
    UseLoopCountLimit=false
   )
   /SetVariables(Emitter.Random_SpinRange:NiagaraFloat=&RandomRangeFloat(Minimum=0, Maximum=1, Random Seed=$Emitter.LoopCount, Fixed Random Seed=true)) {isSetParameters=true}
  ]
  particleSpawnScript [
   &InitializeParticle(
    Color=<rgba 6 1.36 1.2 1>
    Lifetime=&RandomRangeFloat(Minimum=0.25, Maximum=1, Randomness Mode=::NewEnumerator1, Random Seed=$Emitter.LoopCount, Fixed Random Seed=true)
    Ribbon Width=40
    Ribbon Width Mode=::NewEnumerator1
    Sprite Rotation Mode=::NewEnumerator2
    Sprite Size Mode=::NewEnumerator3
    Uniform Sprite Size=100
    UsePositionOffset=false
    Write Material Random=true
   )
   &PartitionParticles(
    Normalized Index In Partition Includes One=true
    Particles Per Partition=10
    Partition Count Mode:ENiagara_InfiniteLoopDuration=::NewEnumerator1
    Partition Mode:ENiagara_PartitionMode=::NewEnumerator1
    Write Partition Index to Ribbon Link Order:ENiagara_AttributeSamplingApplyOutput=::NewEnumerator0
    Write Partition to Ribbon ID=::NewEnumerator0
   )
   &PartitionParticles() {module=PartitionParticles001}
   &ShapeLocation(
    Apply To Particle Position=::NewEnumerator1
    Cylinder Distribution=::NewEnumerator2
    Cylinder Radius=300
    MaskSpawnGroup=false
    Rotation Angle:NiagaraFloat=$Emitter.Random_SpinRange
    Rotation Mode=::NewEnumerator1
    Shape Primitive=::NewEnumerator1
    U Position=&FloatFromCurve(
     FloatCurve=DataInterface() {propertyValues {Curve {keys [<at 0 0>,<at 1 0.15>]}}}
     CurveIndex=&RandomRangeFloat(Minimum=0, Maximum=1, Randomness Mode=::NewEnumerator1, Random Seed=$Emitter.LoopCount, Fixed Random Seed=true)
    )
    UseOwnerScale=true
    V Position=&RandomRangeFloat(Minimum=0, Maximum=1, Randomness Mode=::NewEnumerator1, Random Seed=$Emitter.LoopCount, Fixed Random Seed=true)
   )
   &ShapeLocation(
    Apply To Particle Position=::NewEnumerator1
    Cylinder Distribution=::NewEnumerator2
    Cylinder Radius=300
    MaskSpawnGroup=false
    Rotation Angle:NiagaraFloat=$Emitter.Random_SpinRange
    Rotation Mode=::NewEnumerator1
    Shape Primitive=::NewEnumerator1
    U Position=&FloatFromCurve(
     FloatCurve=DataInterface() {propertyValues {Curve {keys [<at 0 0.175>,<at 1 0.25>]}}}
     CurveIndex=&RandomRangeFloat(Minimum=0, Maximum=1, Randomness Mode=::NewEnumerator1, Random Seed=$Emitter.LoopCount, Fixed Random Seed=true)
    )
    UseOwnerScale=true
    V Position=&RandomRangeFloat(Minimum=0, Maximum=1, Randomness Mode=::NewEnumerator1, Random Seed=$Emitter.LoopCount)
   ) {module=ShapeLocation001}
   /SetVariables(Particles.StartPosition:Vector3f=&Add_Vector(A=&SimulationPosition(), B=$Particles.ShapeLocation.ShapeVector)) {isSetParameters=true}
   /SetVariables(Particles.EndPosition:Vector3f=&Add_Vector(A=&SimulationPosition(), B=$Particles.ShapeLocation001.ShapeVector)) {isSetParameters=true}
   /SetVariables(Particles.MidPoint:Vector3f=&Add_Vector(
     A=&Lerp_Vector(A=$Particles.StartPosition, B=$Particles.EndPosition, Alpha=0.5)
     B=&RandomRangeVector(Minimum=<-50 -50 -50>, Maximum=<50 50 50>, Randomness Mode=::NewEnumerator1, Random Seed=$Emitter.LoopCount, Fixed Random Seed=true)
    )) {isSetParameters=true}
  ]
  particleUpdateScript [
   &ParticleState()
   /SetVariables(Particles.OffsetDistance:NiagaraFloat=&Add_Float(
     A=&RandomRangeFloat(Minimum=100, Maximum=150, Randomness Mode=::NewEnumerator1, Random Seed=$Emitter.LoopCount)
     B=&Lerp_Float(A=0, B=50, Alpha=$Particles.NormalizedAge)
    )) {isSetParameters=true}
   /SetVariables(Particles.MidPoint_Withoffset:Vector3f=&Multiply_VectorByFloat(
     Vector=&NormalizeVector(
      Vector To Normalize=&Subtract_Vector(A=$Particles.MidPoint, B=&SimulationPosition())
      Safe Normalize=true
      Safe Fallback Threshold=0
      Safe Fallback Vector=<0 0 0.001>
     )
     Float=$Particles.OffsetDistance
    )) {isSetParameters=true}
   &self:BezierSpline(
    EndHandle:Vector3f=&Add_Vector(A=&Lerp_Vector(A=$Particles.EndPosition, B=$Particles.MidPoint, Alpha=0.5), B=$Particles.MidPoint_Withoffset)
    EndPosition:Vector3f=$Particles.EndPosition
    PositionAlongSpline:NiagaraFloat=$Particles.PartitionParticles.Normalized Index In Partition
    StartHandle:Vector3f=&Add_Vector(A=&Lerp_Vector(A=$Particles.StartPosition, B=$Particles.MidPoint, Alpha=0.5), B=$Particles.MidPoint_Withoffset)
    StartPosition:Vector3f=$Particles.StartPosition
   ) {module=BezierSpline}
   /SetVariables(Particles.Position=&Add_Vector(
     A=$Particles.Spline_Position
     B=&Multiply_VectorByFloat(
      Vector=&MakeVector(
       X=&RandomRangeFloat(Minimum=0, Maximum=0)
       Y=&RandomRangeFloat(Minimum=0, Maximum=0)
       Z=&RandomRangeFloat(Minimum=0, Maximum=0)
      )
      Float=&FloatFromCurve(
       FloatCurve=DataInterface() {propertyValues {Curve {keys [
           {interpMode=RCIM_Cubic, tangentMode=RCTM_User, time=0, value=0}
           {interpMode=RCIM_Cubic, tangentMode=RCTM_User, time=0.5, value=1}
           {tangentMode=RCTM_User, time=1, value=0}
          ]}}}
       CurveIndex=$Particles.PartitionParticles.Normalized Index In Partition
      )
     )
    )) {isSetParameters=true}
   /SetVariables(Particles.FlarePosition:NiagaraPosition=&MakeCustomVectorFromBool(
     Bool=&SetBoolByIntComparison(A=0, B=$Particles.PartitionParticles001.Partition, Comparison Type=::NewEnumerator4)
     True Vector=$Particles.StartPosition
     False Vector=$Particles.EndPosition
    )) {isSetParameters=true}
   /SetVariables(Particles.FlareColor=&MakeLinearColorFromVectorAndFloat(
     "Vector (RGB)"=&MakeVectorFromLinearColorRGB(Color=<rgba 2 0.4533 0.4 1>)
     "Float (Alpha)"=&Waveform("Waveform [1]"=::NewEnumerator8, "[1] Frequency"=20, "[1] Amplitude Min/Max"=<0 1>)
    )) {isSetParameters=true}
   &JitterPosition(
    Jitter Amount:NiagaraFloat=&FloatFromCurve(
     FloatCurve=DataInterface() {propertyValues {Curve {keys [
         {interpMode=RCIM_Cubic, tangentMode=RCTM_User, time=0, value=0}
         {interpMode=RCIM_Cubic, tangentMode=RCTM_User, time=0.5, value=1}
         {tangentMode=RCTM_User, time=1, value=0}
        ]}}}
     Scale Curve=&RandomRangeFloat(Minimum=50, Maximum=100)
    )
    Jitter Delay:NiagaraFloat=&RandomRangeFloat(Minimum=0.2, Maximum=0.4)
   )
   &ScaleColor(
    Scale RGB=&VectorFromFloat(Value=&FloatFromCurve(FloatCurve=DataInterface() {propertyValues {Curve {keys [<at 0 1>,<at 0.1 20>,<at 0.2 1>,<at 0.3 1>,<at 0.4 10>,<at 0.5 1>,<at 0.6 1>,<at 0.7 1>,<at 0.8 5>,<at 0.9 1>,<at 1 1>]}}}))
    ScaleA=true
    ScaleRGB=true
   )
  ]
  rendererClasses [&NiagaraSpriteRendererProperties, &NiagaraRibbonRendererProperties]
  renderers [
   {rendererClass=&NiagaraSpriteRendererProperties}
   {rendererClass=&NiagaraRibbonRendererProperties, rendererIndex=1}
   {rendererClass=&NiagaraSpriteRendererProperties, rendererIndex=2}
  ]
 }
 helix {
  emitterUpdateScript [
   &EmitterState(Life Cycle Mode=::NewEnumerator1, Loop Behavior=::NewEnumerator1, Loop Delay=0.5, Loop Duration Mode=::NewEnumerator1, UseLoopDelay=false)
   &SpawnBurst_Instantaneous(Spawn Count=160, Use Spawn Probability=false, UseLoopCountLimit=false)
  ]
  particleSpawnScript [&InitializeParticle(
    Color Mode=::NewEnumerator0
    Lifetime=5
    Ribbon Width=10
    Ribbon Width Mode=::NewEnumerator1
    Sprite Size Mode=::NewEnumerator3
    Uniform Sprite Size=5
    UsePositionOffset=false
    Write Material Random=true
   )]
  particleUpdateScript [
   &ParticleState(Kill Particles When Lifetime Has Elapsed=false, Let Infinitely Lived Particles Die When Emitter Deactivates=true)
   &ScaleRibbonWidth()
   &DynamicMaterialParameters(
    Index 0 Param 1=&FloatFromCurve(FloatCurve=DataInterface() {propertyValues {Curve {keys [<at 0 0>,<at 1 1>]}}}, CurveIndex=$Emitter.Age)
    Index 0 Param 4=&Multiply_Float(A=$Emitter.Age, B=-0.2)
    Index 1 Param 1=&FloatFromCurve(FloatCurve=DataInterface() {propertyValues {Curve {keys [<at 0.8 0>,<at 1 1>]}}}, CurveIndex=$System.Overload)
    Index 1 Param 3=&FloatFromCurve(
     FloatCurve=DataInterface() {propertyValues {Curve {keys [<at 0 1>,<at 1 0>]}}}
     CurveIndex=&Frac_Float(Float=&Multiply_Float(A=$Emitter.Age, B=0.2))
    )
    Module.Param0WriteEnabled=true
    Module.Param0WriteEnabled1=true
    Module.Param1WriteEnabled=true
    Module.Param1WriteEnabled1=true
    Module.Param2WriteEnabled=true
    Module.Param2WriteEnabled1=true
    Module.Param3WriteEnabled=true
    Module.Param3WriteEnabled1=true
    Write Parameter Index 1=true
   )
   &ScaleColor(
    Scale Alpha=&FloatFromCurve(
     FloatCurve=DataInterface() {propertyValues {Curve {keys [
         <at 0 0 RCIM_Cubic>
         {interpMode=RCIM_Cubic, tangentMode=RCTM_User, time=0.2, value=1}
         {interpMode=RCIM_Cubic, tangentMode=RCTM_User, time=0.8, value=1}
         <at 1 0 RCIM_Cubic>
        ]}}}
     CurveIndex=&ReturnNormalizedExecIndex()
    )
    Scale RGB=&VectorFromFloat(Value=$System.IntensityBurst)
    ScaleA=true
    ScaleRGB=true
   )
   /SetVariables(
    Particles.AirglowColor:LinearColor=&MultiplyLinearColor(
     Linear Color=&MakeLinearColorFromVectorAndFloat(
      "Vector (RGB)"=<1 1 1>
      "Float (Alpha)"=&FloatFromCurve(
       FloatCurve=DataInterface() {propertyValues {Curve {keys [
           <at 0 0 RCIM_Cubic>
           {interpMode=RCIM_Cubic, tangentMode=RCTM_User, time=0.2, value=1}
           {interpMode=RCIM_Cubic, tangentMode=RCTM_User, time=0.8, value=1}
           <at 1 0 RCIM_Cubic>
          ]}}}
       CurveIndex=&ReturnNormalizedExecIndex()
      )
     )
     Scale Factor=&MakeFloat4FromVectorAndFloat(
      "Vector (RGB)"=&VectorFromFloat(Value=1)
      "Float (Alpha)"=&Lerp_Float(A=0.2, B=0.2, Alpha=$System.Overload)
     )
    )
    Particles.AirglowRibbonWidth:NiagaraFloat=&Multiply_Float(A=$Particles.RibbonWidth, B=2)
   ) {isSetParameters=true, module=SetVariables_#8}
   /SetVariables(
    Particles.Strands_Color:LinearColor=&MultiplyLinearColor(
     Linear Color=&MakeLinearColorFromVectorAndFloat(
      "Vector (RGB)"=<1 1 1>
      "Float (Alpha)"=&Lerp_Float(
       A=&FloatFromCurve(
        FloatCurve=DataInterface() {propertyValues {Curve {keys [
            {interpMode=RCIM_Cubic, tangentMode=RCTM_User, time=0.2, value=1}
            {arriveTangent=-3.4386, interpMode=RCIM_Cubic, leaveTangent=-3.4386, tangentMode=RCTM_User, time=1, value=0}
           ]}}}
        CurveIndex=&ReturnNormalizedExecIndex()
       )
       B=&FloatFromCurve(
        FloatCurve=DataInterface() {propertyValues {Curve {keys [
            {interpMode=RCIM_Cubic, tangentMode=RCTM_User, time=0.8, value=1}
            {arriveTangent=-3.4386, interpMode=RCIM_Cubic, leaveTangent=-3.4386, tangentMode=RCTM_User, time=1, value=0}
           ]}}}
        CurveIndex=&ReturnNormalizedExecIndex()
        Scale Curve=0.5
       )
       Alpha=$System.Overload
      )
     )
     Scale Factor=&MakeFloat4FromVectorAndFloat("Vector (RGB)"=&VectorFromFloat(Value=$System.IntensityBurst), "Float (Alpha)"=1)
    )
    Particles.Strands_RibbonWidth:NiagaraFloat=&Lerp_Float(
     A=&FloatFromCurve(
      FloatCurve=DataInterface() {propertyValues {Curve {keys [<at 0.2 1>, {arriveTangent=-0.625, interpMode=RCIM_Cubic, leaveTangent=-0.625, time=0.5, value=1}, <at 1 0.5 RCIM_Cubic>]}}}
      CurveIndex=&ReturnNormalizedExecIndex()
      Scale Curve=&Lerp_Float(A=20, B=10, Alpha=$System.Overload)
     )
     B=&FloatFromCurve(
      FloatCurve=DataInterface() {propertyValues {Curve {keys [<at 0.2 1>, {arriveTangent=-0.625, interpMode=RCIM_Cubic, leaveTangent=-0.625, time=0.8, value=1}, <at 1 0.5 RCIM_Cubic>]}}}
      CurveIndex=&ReturnNormalizedExecIndex()
      Scale Curve=&Lerp_Float(A=20, B=10, Alpha=$System.Overload)
     )
     Alpha=$System.Overload
    )
   ) {isSetParameters=true, module=SetVariables_#9}
   &UpdateMeshOrientation(
    Facing Direction=$Engine.Owner.SystemXAxis
    Facing Interpolation=false
    Orientation Method=::NewEnumerator1
    Override Quat=false
    Side Direction:Vector3f=$Engine.Owner.SystemYAxis
    Stabilization Mode:ENiagara_OrientVectorCount=::NewEnumerator2
    TransformFacing=false
    TransformSide=false
   )
   &ShapeLocation(
    Additional Quaternion Rotation=true
    Additional Yaw / Pitch / Roll=&MakeVector(
     X=0.25
     Y=&Multiply_Float(A=$Emitter.Age, B=&Lerp_Float(A=0.1, B=0.1, Alpha=$System.Overload))
     Z=0.25
    )
    Cylinder Distribution=::NewEnumerator2
    Cylinder Height=0
    Cylinder Radius=&Lerp_Float(
     A=&FloatFromCurve(
      FloatCurve=DataInterface() {propertyValues {Curve {keys [
          {arriveTangent=-16.4244, interpMode=RCIM_Cubic, leaveTangent=-16.4244, tangentMode=RCTM_User, time=0, value=260}
          {arriveTangent=-157.483, interpMode=RCIM_Cubic, leaveTangent=-157.483, tangentMode=RCTM_User, time=1, value=210}
         ]}}}
      CurveIndex=&ReturnNormalizedExecIndex()
     )
     B=&FloatFromCurve(
      FloatCurve=DataInterface() {propertyValues {Curve {keys [
          {arriveTangent=-15.5821, interpMode=RCIM_Cubic, leaveTangent=-15.5821, tangentMode=RCTM_User, time=0, value=260}
          {arriveTangent=-90.4962, interpMode=RCIM_Cubic, leaveTangent=-90.4962, tangentMode=RCTM_User, time=1, value=200}
         ]}}}
      CurveIndex=&ReturnNormalizedExecIndex()
     )
     Alpha=$System.Overload
    )
    MaskSpawnGroup=false
    Offset Coordinate Space=::Local
    Offset Mode=::NewEnumerator0
    Rotation Mode=::NewEnumerator3
    Rotation Quaternion=$Particles.MeshOrientation
    Shape Primitive=::NewEnumerator1
    U Position=&FloatFromCurve(
     FloatCurve=DataInterface() {propertyValues {Curve {keys [<at 0 0>, {arriveTangent=3.7995, arriveTangentWeight=0.5902, leaveTangent=3.7995, leaveTangentWeight=0.5902, time=1, value=1}]}}}
     CurveIndex=&OneMinusFloat(Float=&ReturnNormalizedExecIndex())
     Scale Curve=&Lerp_Float(A=4, B=6, Alpha=$System.Overload)
    )
    UseOwnerScale=true
   )
   /SetVariables(Particles.RibbonLinkOrder:NiagaraFloat=&ReturnNormalizedExecIndex()) {isSetParameters=true, module=SetVariables_#10}
  ]
  rendererClasses [&NiagaraSpriteRendererProperties, &NiagaraRibbonRendererProperties]
  renderers [
   {rendererClass=&NiagaraSpriteRendererProperties}
   {rendererClass=&NiagaraRibbonRendererProperties, rendererIndex=1}
   {rendererClass=&NiagaraRibbonRendererProperties, rendererIndex=2}
   {rendererClass=&NiagaraRibbonRendererProperties, rendererIndex=3}
  ]
 }
 helix001 {
  emitterUpdateScript [
   &EmitterState(Life Cycle Mode=::NewEnumerator1, Loop Behavior=::NewEnumerator1, Loop Delay=0.5, Loop Duration Mode=::NewEnumerator1, UseLoopDelay=false)
   &SpawnBurst_Instantaneous(Spawn Count=160, Use Spawn Probability=false, UseLoopCountLimit=false)
  ]
  particleSpawnScript [&InitializeParticle(
    Color Mode=::NewEnumerator0
    Lifetime=5
    Ribbon Width=10
    Ribbon Width Mode=::NewEnumerator1
    Sprite Size Mode=::NewEnumerator3
    Uniform Sprite Size=5
    UsePositionOffset=false
    Write Material Random=true
   )]
  particleUpdateScript [
   &ParticleState(Kill Particles When Lifetime Has Elapsed=false, Let Infinitely Lived Particles Die When Emitter Deactivates=true)
   &ScaleRibbonWidth()
   &DynamicMaterialParameters(
    Index 0 Param 1=&FloatFromCurve(FloatCurve=DataInterface() {propertyValues {Curve {keys [<at 0 0>,<at 1 1>]}}}, CurveIndex=$Emitter.Age)
    Index 0 Param 4=&Multiply_Float(A=$Emitter.Age, B=-0.2)
    Index 1 Param 1=&FloatFromCurve(FloatCurve=DataInterface() {propertyValues {Curve {keys [<at 0.8 0>,<at 1 1>]}}}, CurveIndex=$System.Overload)
    Index 1 Param 3=&FloatFromCurve(
     FloatCurve=DataInterface() {propertyValues {Curve {keys [<at 0 1>,<at 1 0>]}}}
     CurveIndex=&Frac_Float(Float=&Multiply_Float(A=$Emitter.Age, B=0.2))
    )
    Module.Param0WriteEnabled=true
    Module.Param0WriteEnabled1=true
    Module.Param1WriteEnabled=true
    Module.Param1WriteEnabled1=true
    Module.Param2WriteEnabled=true
    Module.Param2WriteEnabled1=true
    Module.Param3WriteEnabled=true
    Module.Param3WriteEnabled1=true
    Write Parameter Index 1=true
   )
   &ScaleColor(
    Scale Alpha=&FloatFromCurve(
     FloatCurve=DataInterface() {propertyValues {Curve {keys [
         <at 0 0 RCIM_Cubic>
         {interpMode=RCIM_Cubic, tangentMode=RCTM_User, time=0.2, value=1}
         {interpMode=RCIM_Cubic, tangentMode=RCTM_User, time=0.8, value=1}
         <at 1 0 RCIM_Cubic>
        ]}}}
     CurveIndex=&ReturnNormalizedExecIndex()
    )
    Scale RGB=&VectorFromFloat(Value=1)
    ScaleA=true
    ScaleRGB=true
   )
   /SetVariables(
    Particles.AirglowColor:LinearColor=&MultiplyLinearColor(
     Linear Color=&MakeLinearColorFromVectorAndFloat(
      "Vector (RGB)"=<1 1 1>
      "Float (Alpha)"=&FloatFromCurve(
       FloatCurve=DataInterface() {propertyValues {Curve {keys [
           <at 0 0 RCIM_Cubic>
           {interpMode=RCIM_Cubic, tangentMode=RCTM_User, time=0.2, value=1}
           {interpMode=RCIM_Cubic, tangentMode=RCTM_User, time=0.8, value=1}
           <at 1 0 RCIM_Cubic>
          ]}}}
       CurveIndex=&ReturnNormalizedExecIndex()
      )
     )
     Scale Factor=&MakeFloat4FromVectorAndFloat(
      "Vector (RGB)"=&VectorFromFloat(Value=1)
      "Float (Alpha)"=&Lerp_Float(A=0.2, B=0.2, Alpha=$System.Overload)
     )
    )
    Particles.AirglowRibbonWidth:NiagaraFloat=&Multiply_Float(A=$Particles.RibbonWidth, B=2)
   ) {isSetParameters=true, module=SetVariables_#8}
   /SetVariables(
    Particles.Strands_Color:LinearColor=&MultiplyLinearColor(
     Linear Color=&MakeLinearColorFromVectorAndFloat(
      "Vector (RGB)"=&VectorFromFloat(Value=1)
      "Float (Alpha)"=&Lerp_Float(
       A=&FloatFromCurve(
        FloatCurve=DataInterface() {propertyValues {Curve {keys [
            {interpMode=RCIM_Cubic, tangentMode=RCTM_User, time=0.2, value=1}
            {arriveTangent=-3.4386, interpMode=RCIM_Cubic, leaveTangent=-3.4386, tangentMode=RCTM_User, time=1, value=0}
           ]}}}
        CurveIndex=&ReturnNormalizedExecIndex()
       )
       B=&FloatFromCurve(
        FloatCurve=DataInterface() {propertyValues {Curve {keys [
            {interpMode=RCIM_Cubic, tangentMode=RCTM_User, time=0.8, value=1}
            {arriveTangent=-3.4386, interpMode=RCIM_Cubic, leaveTangent=-3.4386, tangentMode=RCTM_User, time=1, value=0}
           ]}}}
        CurveIndex=&ReturnNormalizedExecIndex()
        Scale Curve=0.5
       )
       Alpha=$System.Overload
      )
     )
     Scale Factor=&MakeFloat4FromVectorAndFloat("Vector (RGB)"=&VectorFromFloat(Value=$System.IntensityBurst), "Float (Alpha)"=1)
    )
    Particles.Strands_RibbonWidth:NiagaraFloat=&Lerp_Float(
     A=&FloatFromCurve(
      FloatCurve=DataInterface() {propertyValues {Curve {keys [<at 0.2 1>, {arriveTangent=-0.625, interpMode=RCIM_Cubic, leaveTangent=-0.625, time=0.5, value=1}, <at 1 0.5 RCIM_Cubic>]}}}
      CurveIndex=&ReturnNormalizedExecIndex()
      Scale Curve=&Lerp_Float(A=20, B=10, Alpha=$System.Overload)
     )
     B=&FloatFromCurve(
      FloatCurve=DataInterface() {propertyValues {Curve {keys [
          <at 0 0 RCIM_Cubic>
          {interpMode=RCIM_Cubic, tangentMode=RCTM_User, time=0.2, value=1}
          {interpMode=RCIM_Cubic, tangentMode=RCTM_User, time=0.8, value=1}
          <at 1 0.5 RCIM_Cubic>
         ]}}}
      CurveIndex=&ReturnNormalizedExecIndex()
      Scale Curve=&Lerp_Float(A=20, B=10, Alpha=$System.Overload)
     )
     Alpha=$System.Overload
    )
   ) {isSetParameters=true, module=SetVariables_#9}
   &UpdateMeshOrientation(
    Facing Direction=$Engine.Owner.SystemXAxis
    Facing Interpolation=false
    Orientation Method=::NewEnumerator1
    Override Quat=false
    Side Direction:Vector3f=$Engine.Owner.SystemYAxis
    Stabilization Mode:ENiagara_OrientVectorCount=::NewEnumerator2
    TransformFacing=false
    TransformSide=false
   )
   &ShapeLocation(
    Additional Quaternion Rotation=true
    Additional Yaw / Pitch / Roll=&MakeVector(
     X=0.25
     Y=&Multiply_Float(A=$Emitter.Age, B=&Lerp_Float(A=0.1, B=0.5, Alpha=$System.Overload))
     Z=0.25
    )
    Cylinder Distribution=::NewEnumerator2
    Cylinder Height=0
    Cylinder Radius=&Lerp_Float(
     A=&FloatFromCurve(
      FloatCurve=DataInterface() {propertyValues {Curve {keys [
          {arriveTangent=-16.4244, interpMode=RCIM_Cubic, leaveTangent=-16.4244, tangentMode=RCTM_User, time=0, value=260}
          {arriveTangent=-250.2426, interpMode=RCIM_Cubic, leaveTangent=-250.2426, tangentMode=RCTM_User, time=1, value=160}
         ]}}}
      CurveIndex=&ReturnNormalizedExecIndex()
     )
     B=&FloatFromCurve(
      FloatCurve=DataInterface() {propertyValues {Curve {keys [
          {arriveTangent=-21.2235, interpMode=RCIM_Cubic, leaveTangent=-21.2235, tangentMode=RCTM_User, time=0, value=230}
          {arriveTangent=-51.5016, interpMode=RCIM_Cubic, leaveTangent=-51.5016, tangentMode=RCTM_User, time=1, value=190}
         ]}}}
      CurveIndex=&ReturnNormalizedExecIndex()
     )
     Alpha=$System.Overload
    )
    MaskSpawnGroup=false
    Offset Coordinate Space=::Local
    Offset Mode=::NewEnumerator0
    Rotation Mode=::NewEnumerator3
    Rotation Quaternion=$Particles.MeshOrientation
    Shape Primitive=::NewEnumerator1
    U Position=&FloatFromCurve(
     FloatCurve=DataInterface() {propertyValues {Curve {keys [<at 0 0>, {arriveTangent=3.7995, arriveTangentWeight=0.5902, leaveTangent=3.7995, leaveTangentWeight=0.5902, time=1, value=1}]}}}
     CurveIndex=&OneMinusFloat(Float=&ReturnNormalizedExecIndex())
     Scale Curve=&Lerp_Float(A=4, B=2, Alpha=$System.Overload)
    )
    UseOwnerScale=true
   )
   /SetVariables(Particles.RibbonLinkOrder:NiagaraFloat=&ReturnNormalizedExecIndex()) {isSetParameters=true, module=SetVariables_#10}
  ]
  rendererClasses [&NiagaraSpriteRendererProperties, &NiagaraRibbonRendererProperties]
  renderers [
   {rendererClass=&NiagaraSpriteRendererProperties}
   {rendererClass=&NiagaraRibbonRendererProperties, rendererIndex=1}
   {rendererClass=&NiagaraRibbonRendererProperties, rendererIndex=2}
   {rendererClass=&NiagaraRibbonRendererProperties, rendererIndex=3}
  ]
 }
}
summary {
 systemName=NS_RiftGate_Tendrils
 emitters [
  {emitterName=helix, rendererClasses [&NiagaraSpriteRendererProperties, &NiagaraRibbonRendererProperties]}
  {emitterName=Dots, rendererClasses [&NiagaraSpriteRendererProperties]}
  {emitterName=Lights, rendererClasses [&NiagaraSpriteRendererProperties, &NiagaraLightRendererProperties]}
  {emitterName=EnergyMesh001, rendererClasses [&NiagaraMeshRendererProperties]}
  {
   emitterName=ContactRibbons_Interp
   rendererClasses [&NiagaraSpriteRendererProperties, &NiagaraRibbonRendererProperties]
  }
  {
   emitterName=LightningMesh
   rendererClasses [
    &NiagaraMeshRendererProperties
    &NiagaraSpriteRendererProperties
    &NiagaraLightRendererProperties
   ]
  }
  {emitterName=helix001, rendererClasses [&NiagaraSpriteRendererProperties, &NiagaraRibbonRendererProperties]}
  {emitterName=LightCone_mesh, rendererClasses [&NiagaraMeshRendererProperties]}
  {emitterName=Tesla, rendererClasses [&NiagaraSpriteRendererProperties, &NiagaraRibbonRendererProperties]}
  {emitterName=Embers_Trapped, rendererClasses [&NiagaraSpriteRendererProperties]}
  {
   emitterName=LightningMesh001
   rendererClasses [
    &NiagaraMeshRendererProperties
    &NiagaraSpriteRendererProperties
    &NiagaraLightRendererProperties
   ]
  }
  {emitterName=Embers_Flare, rendererClasses [&NiagaraSpriteRendererProperties]}
  {emitterName=Flash, rendererClasses [&NiagaraMeshRendererProperties]}
 ]
 userVariables [
  User.FlickerOff {type=NiagaraBool}
  User.Helix_Height {type=NiagaraFloat, defaultValue=1, description=""}
  User.Helix_Mirrored {type=NiagaraBool, defaultValue=true}
  User.Helix_OffsetRotation {type=NiagaraFloat, defaultValue=0, description=""}
  User.HelixRadius_Bottom {type=NiagaraFloat, defaultValue=1, description=""}
  User.HelixRadius_Top {type=NiagaraFloat, defaultValue=0, description=""}
  User.IntensityBurst {type=NiagaraBool}
  User.New Skeletal Mesh {
   type=NiagaraDataInterfaceSkeletalMesh
   defaultValue {
    struct=NiagaraExt_VariableValue_DataInterface
    value {dataInterface=&self:NiagaraDataInterfaceSkeletalMesh_1}
   }
  }
  User.Overload {type=NiagaraFloat, defaultValue=0, description=""}
  User.Pulled_0to1 {type=NiagaraFloat, defaultValue=0, description=""}
  User.SkeletalMesh {
   type=NiagaraDataInterfaceSkeletalMesh
   defaultValue {
    struct=NiagaraExt_VariableValue_DataInterface
    value {dataInterface=&self:NiagaraDataInterfaceSkeletalMesh_0}
   }
  }
  User.Speed_Lerp {type=NiagaraFloat, defaultValue=0, description=""}
  User.TightenGrip {type=NiagaraFloat, defaultValue=0, description=""}
  User.ZP_Pos {type=Vector3f, description="", defaultValue {struct=Vector3f, value=<2766.7109 582.1028 24366.7988>}}
 ]
}
